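; Anim notify state for the F_MED_Bistro_Monster cosmetic that swaps player pawn materials for the duration of the notify.
; Begin creates dynamic instances for four parts, but SetPartMaterials applies only the tail one (Body_Transition) to MiscOrTail slot 0.
; End optionally restores the tail to the Body material; both events print debug strings.
; TailWipe ramps MaterialCurveA over time and clears its timer, but nothing shown calls it.
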(fn Received_NotifyBegin
  (params (MeshComp :type Skeletal%20Mesh%20Component%20Object%20Reference) (Animation :type Anim%20Sequence%20Base%20Object%20Reference) (TotalDuration :type Float%20%28single-precision%29) (EventReference :type Anim%20Notify%20Event%20Reference%20Structure%20%28by%20ref%29))
  (bind _asfort_player_pawn (Utilities|Casting|CastToFortPlayerPawn (Components|GetOwner MeshComp)))
  (CallFunction|SetPartMaterialsDynamic _asfort_player_pawn "/BRCosmetics/Characters/Player/Female/Medium/Bodies/F_MED_Bistro_Monster/Materials/F_MED_Bistro_Monster_Hair.F_MED_Bistro_Monster_Hair" "/BRCosmetics/Characters/Player/Female/Medium/Bodies/F_MED_Bistro_Monster/Materials/F_MED_Bistro_Monster_Head.F_MED_Bistro_Monster_Head" "/BRCosmetics/Characters/Player/Female/Medium/Bodies/F_MED_Bistro_Monster/Materials/F_MED_Bistro_Monster_Body.F_MED_Bistro_Monster_Body" "/BRCosmetics/Characters/Player/Female/Medium/Bodies/F_MED_Bistro_Monster/Materials/F_MED_Bistro_Monster_Body_Transition.F_MED_Bistro_Monster_Body_Transition")
  (Development|PrintString 0 "SET MATERIALS" true true "(R=0.000000,G=1.000000,B=0.161651,A=1.000000)"))

(fn Received_NotifyEnd
  (params (MeshComp :type Skeletal%20Mesh%20Component%20Object%20Reference) (Animation :type Anim%20Sequence%20Base%20Object%20Reference) (EventReference :type Anim%20Notify%20Event%20Reference%20Structure%20%28by%20ref%29))
  (if (Variables|Default|GetReset%20Materials%20When%20Done)
    (bind _asfort_player_pawn (Utilities|Casting|CastToFortPlayerPawn (Components|GetOwner MeshComp)))
    (CallFunction|SetPartMaterials _asfort_player_pawn "/BRCosmetics/Characters/Player/Female/Medium/Bodies/F_MED_Bistro_Monster/Materials/F_MED_Bistro_Monster_Body.F_MED_Bistro_Monster_Body" "/BRCosmetics/Characters/Player/Female/Medium/Bodies/F_MED_Bistro_Monster/Materials/F_MED_Bistro_Monster_Hair.F_MED_Bistro_Monster_Hair" "/BRCosmetics/Characters/Player/Female/Medium/Bodies/F_MED_Bistro_Monster/Materials/F_MED_Bistro_Monster_Body.F_MED_Bistro_Monster_Body" "/BRCosmetics/Characters/Player/Female/Medium/Bodies/F_MED_Bistro_Monster/Materials/F_MED_Bistro_Monster_Body.F_MED_Bistro_Monster_Body")
    (Development|PrintString 0 "RESTORE")))

(fn SetPartMaterials
  (params (Pawn :type Fort%20Player%20Pawn%20Object%20Reference) (Face%20Material :type Material%20Instance%20Object%20Reference) (Head%20Material :type Material%20Instance%20Object%20Reference) (Body%20Material :type Material%20Interface%20Object%20Reference) (Tail%20Material :type Material%20Interface%20Object%20Reference))
  (Utilities|IsValid Pawn
    (:Is%20Valid
      (Utilities|FlowControl|Sequence
        (:then_3
          (Rendering|Material|SetMaterial (Customization|GetSkeletalMeshforPartType Pawn "MiscOrTail" :!class /Script/FortniteGame.FortPlayerPawn) 0 Tail%20Material))))))

(fn SetPartMaterialsDynamic
  (params (Player%20Pawn :type Fort%20Player%20Pawn%20Object%20Reference) (Face%20Material%20Base :type Material%20Interface%20Object%20Reference) (Head%20Material%20Base :type Material%20Interface%20Object%20Reference) (Body%20Material%20Base :type Material%20Interface%20Object%20Reference) (Tail%20Material%20Base :type Material%20Interface%20Object%20Reference))
  (bind _returnvalue (Rendering|Material|CreateDynamicMaterialInstance Player%20Pawn Face%20Material%20Base "" "None" :!class /Script/Engine.KismetMaterialLibrary))
  (bind _returnvalue_2 (Rendering|Material|CreateDynamicMaterialInstance Player%20Pawn Head%20Material%20Base "" "None" :!class /Script/Engine.KismetMaterialLibrary))
  (bind _returnvalue_3 (Rendering|Material|CreateDynamicMaterialInstance Player%20Pawn Body%20Material%20Base "" "None" :!class /Script/Engine.KismetMaterialLibrary))
  (bind _returnvalue_4 (Rendering|Material|CreateDynamicMaterialInstance Player%20Pawn Tail%20Material%20Base "" "None" :!class /Script/Engine.KismetMaterialLibrary))
  (CallFunction|SetPartMaterials Player%20Pawn _returnvalue _returnvalue_2 _returnvalue_3 _returnvalue_4))

(fn TailWipe
  (params (Start%20Time :type Float%20%28double-precision%29) (TailTransitionTime :type Float%20%28double-precision%29) (Material :type Material%20Instance%20Dynamic%20Object%20Reference))
  (bind _returnvalue (Utilities|Time|GetGameTimeinSeconds Material))
  (bind _returnvalue_2 (- _returnvalue Start%20Time))
  (bind _returnvalue_3 (/ _returnvalue_2 TailTransitionTime))
  (bind _returnvalue_4 (Math|Float|Lerp 0.0 1.0 _returnvalue_3))
  (Rendering|Material|SetScalarParameterValue Material "MaterialCurveA" _returnvalue_4 :!class /Script/Engine.MaterialInstanceDynamic)
  (if (>= _returnvalue_4 1.0)
    (Utilities|Time|ClearTimerbyFunctionName Material "TailWipe")))
