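# Material function for the Bistro Monster female character body that drives a transition effect on incoming material attributes.
# BistroColorCurve blends the input emissive toward the Monster Emissive Color parameter.
# BistroTransitionCurve is remapped from -1 to 1, added to the TransitionMask input and clamped to that range to replace Opacity Mask.
# A Saturate node on the same sum is left unconnected to the output.
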MaterialFunction /BRCosmetics/Characters/Player/Female/Medium/Bodies/F_MED_Bistro_Monster/Materials/MF_Bistro_Monster_Transition
graph {nodes [
  n1 = GetMaterialAttributes(None=FunctionInput() {inputName=Input, inputType=FunctionInput_MaterialAttributes}) {attributeGetTypes [#1]}
  n2 = VectorParameter() {parameterName=Monster Emissive Color, defaultValue=<rgba 150 0 8.2096 1>}
  n3 = ScalarParameter() {parameterName=BistroColorCurve}
  n4 = Add(
   A=LinearInterpolate(Alpha=ScalarParameter() {parameterName=BistroTransitionCurve}) {constA=-1}
   B=FunctionInput() {inputName=TransitionMask, inputType=FunctionInput_Scalar}
  )
  FunctionOutput(None=SetMaterialAttributes!(
    MaterialAttributes=@n1.MaterialAttributes
    EmissiveColor=LinearInterpolate(A=@n1.EmissiveColor, B=@n2.RGB, Alpha=@n3)
    OpacityMask=Clamp(None=@n4) {minDefault=-1}
   ) {attributeSetTypes [#1, #2]})
  Saturate(None=@n4)
 ]}
