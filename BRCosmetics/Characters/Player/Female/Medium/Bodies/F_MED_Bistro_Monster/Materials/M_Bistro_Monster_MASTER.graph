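# Master character material for the medium female Bistro Monster body cosmetic, built on material attributes with masked, dithered opacity.
# The shared character PreFX function feeds a chain of thin film, monster transition, panning emissive, metals LUT and emissive fresnel functions, ending in PostFX.
# A Skin FX mask texture drives the effects: red for metals, green for thin film and panning emissive, blue for the transition.
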Material /BRCosmetics/Characters/Player/Female/Medium/Bodies/F_MED_Bistro_Monster/Materials/M_Bistro_Monster_MASTER
graph {
 properties {
  bUseMaterialAttributes=true
  bUsedWithMeshDeformer=true
  bUsedWithMorphTargets=true
  bUsedWithSkeletalMesh=true
  blendMode=BLEND_Masked
  ditherOpacityMask=true
  pixelDepthOffsetMode=PDOM_Legacy
  usedShadingModels=Default Lit
 }
 nodes [
  n1 = MaterialFunctionCall() {materialFunction=&BaseFN_CharacterFunction_PreFX}
  n2 = TextureSampleParameter2D() {
   parameterName=Skin FX
   samplerType=SAMPLERTYPE_Masks
   texture=/BRCosmetics/Characters/Player/Female/Medium/Bodies/F_MED_Bistro_Monster/Textures/T_F_MED_Bistro_Monster_Body_FX
  }
  n3 = MaterialFunctionCall!(
   Alpha=@n2.G
   isVehicle
   Gmap=@n1.Gmap
   UVOffset
   Local Normal
   Local Position=@n1.Gmap
   EmissiveColorOverride
   Material=MaterialFunctionCall!(
    TransitionMask=@n2.B
    Input=MaterialFunctionCall!(Alpha=@n2.G, Material Attributes=@n1.Result) {materialFunction=/Game/Weapons/WeaponSkins/MaterialFunction/ThinFilmFunction}.Result
   ) {materialFunction=&/MF_Bistro_Monster_Transition}.Result
  ) {materialFunction=/Game/Weapons/WeaponSkins/MaterialFunction/PanningEmissive}
  MaterialFunctionCall(
   AO=@n1.AO
   Distance Blend=@n1.AO
   Fresnel=@n1.AO
   Gmap=@n1.AO
   Pre Skinned Local Position=@n1.AO
   Material=MaterialFunctionCall!(In=MaterialFunctionCall(Mask=@n2.R, Materials=@n3.Result) {materialFunction=&MF_MetalsLUT}.Result) {materialFunction=/Game/Weapons/WeaponSkins/MaterialFunction/MF_EmissiveFresnel}.Result
  ) {materialFunction=&BaseFN_CharacterFunction_PostFX}
 ]
}
