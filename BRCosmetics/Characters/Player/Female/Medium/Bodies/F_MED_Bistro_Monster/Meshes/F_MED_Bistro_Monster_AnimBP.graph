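# Anim Blueprint for the F_MED_Bistro_Monster cosmetic body, correcting shoulder deformation and muscle volume atop the parent mesh's pose.
# Weighted constraints drive dyn_shldr bones, then additive offset chains, chosen by emote and skydiving state, scale with arm and neck rotation.
# A muscle-volume Control Rig finishes the pose using per-frame limb deltas from ref pose and arm pose-deformation queries.
# On customization load, it hides the tail when a long backbling is worn.
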AnimBlueprint /BRCosmetics/Characters/Player/Female/Medium/Bodies/F_MED_Bistro_Monster/Meshes/F_MED_Bistro_Monster_AnimBP
target_skeleton=&/F_MED_Bistro_Monster_Skeleton
graphs [
 AnimGraph {nodes [
   n1 = "Transform (Modify) Bone - Bone: dyn_shldr_R_high"(
    ComponentPose=Local To Component(LocalPose=Use cached pose 'Pre-sim'())
    Alpha=Get Neck_Back_Rot() -> (Neck_Back_Rot:double)
    Translation="-1,0,0"
    Rotation="25,0,0"
    Scale="1,1,1"
   ) {
    rotationMode=BMM_Additive
    rotationSpace=BCS_BoneSpace
    translationMode=BMM_Additive
    translationSpace=BCS_BoneSpace
    alphaBoolBlend {blendOption=Linear}
    alphaScaleBiasClamp {bInterpResult=true, inRange {max=65, min=50}}
    boneToModify {boneName=dyn_shldr_R_high}
    rotation {pitch=25}
    translation {x=-1}
   }
   n2 = Get UpperArm_R_rot() -> (UpperArm_R_rot:double)
   n3 = "Transform (Modify) Bone - Bone: dyn_shldr_R_high"(ComponentPose=@n1, Alpha=@n2, Translation="0,0,0", Rotation="-15,0,0", Scale="1,1,1") {
    rotationMode=BMM_Additive
    rotationSpace=BCS_BoneSpace
    translationMode=BMM_Additive
    translationSpace=BCS_BoneSpace
    alphaBoolBlend {blendOption=Linear}
    alphaScaleBiasClamp {bInterpResult=true, bMapRange=true, inRange {max=40, min=15}}
    boneToModify {boneName=dyn_shldr_R_high}
    rotation {pitch=-15}
   }
   n4 = "Transform (Modify) Bone - Bone: dyn_shldr_R_low"(ComponentPose=@n3, Alpha=@n2, Translation="0,0,0", Rotation="20,0,0", Scale="1,1,1") {
    rotationMode=BMM_Additive
    rotationSpace=BCS_BoneSpace
    translationMode=BMM_Additive
    translationSpace=BCS_BoneSpace
    alphaBoolBlend {blendOption=Linear}
    alphaScaleBiasClamp {bInterpResult=true, bMapRange=true, inRange {max=40, min=15}}
    boneToModify {boneName=dyn_shldr_R_low}
    rotation {pitch=20}
   }
   n5 = "Transform (Modify) Bone - Bone: dyn_shldr_L_high"(ComponentPose=@n4, Alpha=Get UpperArm_L _rot() -> (UpperArm_L _rot:double), Translation="-1,0,0", Rotation="15,0,0", Scale="1,1,1") {
    rotationMode=BMM_Additive
    rotationSpace=BCS_BoneSpace
    translationMode=BMM_Additive
    translationSpace=BCS_BoneSpace
    alphaBoolBlend {blendOption=Linear}
    alphaScaleBiasClamp {bInterpResult=true, bMapRange=true, inRange {max=40, min=15}}
    boneToModify {boneName=dyn_shldr_L_high}
    rotation {pitch=15}
    translation {x=-1}
   }
   n6 = "Transform (Modify) Bone - Bone: dyn_shldr_L_low"(ComponentPose=@n5, Alpha=@n2, Translation="0,0,0", Rotation="-20,0,0", Scale="1,1,1") {
    rotationMode=BMM_Additive
    rotationSpace=BCS_BoneSpace
    translationMode=BMM_Additive
    translationSpace=BCS_BoneSpace
    alphaBoolBlend {blendOption=Linear}
    alphaScaleBiasClamp {bInterpResult=true, bMapRange=true, inRange {max=40, min=15}}
    boneToModify {boneName=dyn_shldr_L_low}
    rotation {pitch=-20}
   }
   n7 = Component To Local(ComponentPose=@n6)
   n8 = Get UpperArm_R_rot()
   n9 = "Transform (Modify) Bone - Bone: dyn_shldr_R_high"(ComponentPose=Local To Component(LocalPose=Use cached pose 'Pre-sim'()), Alpha=@n8, Translation="1,0,0", Rotation="-15,0,0", Scale="1,1,1") {
    rotationMode=BMM_Additive
    rotationSpace=BCS_BoneSpace
    translationMode=BMM_Additive
    translationSpace=BCS_BoneSpace
    alphaBoolBlend {blendOption=Linear}
    alphaScaleBiasClamp {bInterpResult=true, bMapRange=true, inRange {max=65, min=50}}
    boneToModify {boneName=dyn_shldr_R_high}
    rotation {pitch=-15}
    translation {x=1}
   }
   n10 = "Transform (Modify) Bone - Bone: dyn_shldr_R_low"(ComponentPose=@n9, Alpha=@n8, Translation="0,0,0", Rotation="20,0.000,0.000", Scale="1,1,1") {
    rotationMode=BMM_Additive
    rotationSpace=BCS_BoneSpace
    translationMode=BMM_Additive
    translationSpace=BCS_BoneSpace
    alphaBoolBlend {blendOption=Linear}
    alphaScaleBiasClamp {bInterpResult=true, bMapRange=true, inRange {max=65, min=50}}
    boneToModify {boneName=dyn_shldr_R_low}
    rotation {pitch=20}
   }
   n11 = Get UpperArm_L _rot()
   n12 = "Transform (Modify) Bone - Bone: dyn_shldr_L_high"(ComponentPose=@n10, Alpha=@n11, Translation="-1,0,0", Rotation="15,0,0", Scale="1,1,1") {
    rotationMode=BMM_Additive
    rotationSpace=BCS_BoneSpace
    translationMode=BMM_Additive
    translationSpace=BCS_BoneSpace
    alphaBoolBlend {blendOption=Linear}
    alphaScaleBiasClamp {bInterpResult=true, bMapRange=true, inRange {max=65, min=50}}
    boneToModify {boneName=dyn_shldr_L_high}
    rotation {pitch=15}
    translation {x=-1}
   }
   n13 = "Transform (Modify) Bone - Bone: dyn_shldr_L_low"(ComponentPose=@n12, Alpha=@n11, Translation="0,0,0", Rotation="-20,0,0", Scale="1,1,1") {
    rotationMode=BMM_Additive
    rotationSpace=BCS_BoneSpace
    translationMode=BMM_Additive
    translationSpace=BCS_BoneSpace
    alphaBoolBlend {blendOption=Linear}
    alphaScaleBiasClamp {bInterpResult=true, bMapRange=true, inRange {max=65, min=50}}
    boneToModify {boneName=dyn_shldr_L_low}
    rotation {pitch=-20}
   }
   n14 = Component To Local(ComponentPose=@n13)
   n15 = Get bIsSkyDiving()
   n16 = Get UpperArm_R_rot()
   n17 = "Transform (Modify) Bone - Bone: dyn_shldr_R_high"(
    ComponentPose=Local To Component(LocalPose=Use cached pose 'Pre-sim'())
    Alpha=@n16
    Translation="1.5,0,0"
    Rotation="-30,0,0"
    Scale="1,1,1"
   ) {
    rotationMode=BMM_Additive
    rotationSpace=BCS_BoneSpace
    translationMode=BMM_Additive
    translationSpace=BCS_BoneSpace
    alphaBoolBlend {blendOption=Linear}
    alphaScaleBiasClamp {bInterpResult=true, bMapRange=true, inRange {max=50, min=15}}
    boneToModify {boneName=dyn_shldr_R_high}
    rotation {pitch=-30}
    translation {x=1.5}
   }
   n18 = "Transform (Modify) Bone - Bone: dyn_shldr_R_low"(ComponentPose=@n17, Alpha=@n16, Translation="0,0,0", Rotation="15,0.000,0.000", Scale="1,1,1") {
    rotationMode=BMM_Additive
    rotationSpace=BCS_BoneSpace
    translationMode=BMM_Additive
    translationSpace=BCS_BoneSpace
    alphaBoolBlend {blendOption=Linear}
    alphaScaleBiasClamp {bInterpResult=true, bMapRange=true, inRange {max=50, min=15}}
    boneToModify {boneName=dyn_shldr_R_low}
    rotation {pitch=15}
   }
   n19 = Get UpperArm_L _rot()
   n20 = "Transform (Modify) Bone - Bone: dyn_shldr_L_high"(ComponentPose=@n18, Alpha=@n19, Translation="-1.5,0,0", Rotation="30,0,0", Scale="1,1,1") {
    rotationMode=BMM_Additive
    rotationSpace=BCS_BoneSpace
    translationMode=BMM_Additive
    translationSpace=BCS_BoneSpace
    alphaBoolBlend {blendOption=Linear}
    alphaScaleBiasClamp {bInterpResult=true, bMapRange=true, inRange {max=50, min=15}}
    boneToModify {boneName=dyn_shldr_L_high}
    rotation {pitch=30}
    translation {x=-1.5}
   }
   n21 = "Transform (Modify) Bone - Bone: dyn_shldr_L_low"(ComponentPose=@n20, Alpha=@n19, Translation="0,0,0", Rotation="-15,0,0", Scale="1,1,1") {
    rotationMode=BMM_Additive
    rotationSpace=BCS_BoneSpace
    translationMode=BMM_Additive
    translationSpace=BCS_BoneSpace
    alphaBoolBlend {blendOption=Linear}
    alphaScaleBiasClamp {bInterpResult=true, bMapRange=true, inRange {max=50, min=15}}
    boneToModify {boneName=dyn_shldr_L_low}
    rotation {pitch=-15}
   }
   n22 = Break Pose Deformation Query Results Arms!(PoseDeformationQueryResults_Arms:"Pose Deformation Query Results Arms Structure (by ref)"=Break Pose Deformation Query Arms!(PoseDeformationQuery_Arms:"Pose Deformation Query Arms Structure (by ref)"=Get Arms() -> (Arms:Pose Deformation Query Arms Structure)) -> (QueryResults:Pose Deformation Query Results Arms Structure) {is=title K2Node_BreakStruct}) -> (
    ArmUpL:float=~
    ArmUpR:float=~
    ArmForwardL:float=~
    ArmForwardR:float=~
    ArmBackL:float=~
    ArmBackR:float=~
    ArmForwardUpL:float=~
    ArmForwardUpR:float=~
    ArmForwardDownL:float=~
    ArmForwardDownR:float=~
   ) {is=title K2Node_BreakStruct}
   n23 = Control Rig!(
    Source=Blend Poses by bool(
     BlendPose_0=Component To Local(ComponentPose=@n21)
     BlendPose_1=Blend Poses by bool(BlendPose_0=@n14, BlendPose_1=@n7, bActiveValue=@n15) {blendType=Linear}
     bActiveValue=Get bIsPlayingEmote()
    ) {blendType=Linear}
    Alpha
    thigh_rot_l:double=Get thigh_rot_l()
    elbow_rot_l:double=Get elbow_rot_l() -> (elbow_rot_l:double)
    elbow_rot_r:double=Get elbow_rot_r() -> (elbow_rot_r:double)
    thigh_rot_r:double=Get thigh_rot_r()
    knee_rot_l:double=Get knee_rot_l() -> (knee_rot_l:double)
    knee_rot_r:double=Get knee_rot_r() -> (knee_rot_r:double)
    psd_arm_up_alpha_l:double=@n22
    psd_arm_up_alpha_r:double=@n22.ArmUpR
    psd_arm_fwd_alpha_l:double=@n22.ArmForwardL
    psd_arm_fwd_alpha_r:double=@n22.ArmForwardR
    psd_arm_bk_alpha_l:double=@n22.ArmBackL
    psd_arm_bk_alpha_r:double=@n22.ArmBackR
    psd_arm_fwd_up_alpha_l:double=@n22.ArmForwardUpL
    psd_arm_fwd_up_alpha_r:double=@n22.ArmForwardUpR
    psd_arm_fwd_dn_alpha_l:double=@n22.ArmForwardDownL
    psd_arm_fwd_dn_alpha_r:double=@n22.ArmForwardDownR
   ) -> (Pose) {
    alphaBoolBlend {blendOption=Linear}
    controlRigAssetReference {blueprintRigClass=&/F_MED_Bistro_Monster_Muscle_Volume_ControlRig.F_MED_Bistro_Monster_Muscle_Volume_ControlRig_C}
   }
   Output Pose(Result=@n23) {layerGroup=None}
   n24 = "Constraint - dyn_shldr_L_MSTR (PPPPPP)"!(
    ComponentPose=Local To Component(LocalPose=Copy Pose From Mesh() {bCopyCurves=true, bUseAttachedParent=true})
    Alpha=1
    ConstraintWeights_0=0.7938
    ConstraintWeights_1:float=0.0051
    ConstraintWeights_2:float=0.1568
    ConstraintWeights_3:float=0.021
    ConstraintWeights_4:float=0.0198
    ConstraintWeights_5:float=0.0035
   ) -> (Pose) {
    alphaBoolBlend {blendOption=Linear}
    boneToModify {boneName=dyn_shldr_L_MSTR}
    constraintSetup [
     {targetBone {boneName=clavicle_l}}
     {targetBone {boneName=spine_04}}
     {targetBone {boneName=spine_05}}
     {targetBone {boneName=upperarm_twist_01_l}}
     {targetBone {boneName=upperarm_up_volume_l}}
     {targetBone {boneName=upperarm_fr_volume_l}}
    ]
    constraintWeights [1,1,1,1,1,1]
   }
   n25 = "Constraint - dyn_shldr_L_high (PPPPPPPP)"!(
    ComponentPose=@n24
    Alpha=1
    ConstraintWeights_0=0.6699
    ConstraintWeights_1:float=0
    ConstraintWeights_2:float=0.0048
    ConstraintWeights_3:float=0.3252
    ConstraintWeights_4:float=0
    ConstraintWeights_5:float=0
    ConstraintWeights_6:float=0
    ConstraintWeights_7:float=0
   ) -> (Pose) {
    alphaBoolBlend {blendOption=Linear}
    boneToModify {boneName=dyn_shldr_L_high}
    constraintSetup [
     {targetBone {boneName=clavicle_l}}
     {targetBone {boneName=spine_03}}
     {targetBone {boneName=spine_04}}
     {targetBone {boneName=spine_05}}
     {targetBone {boneName=upperarm_twist_01_l}}
     {targetBone {boneName=neck_02}}
     {targetBone {boneName=neck_01}}
     {targetBone {boneName=upperarm_up_volume_l}}
    ]
    constraintWeights [1,1,1,1,1,1,1,1]
   }
   n26 = "Constraint - dyn_shldr_L_low (PPPPPPP)"!(
    ComponentPose=@n25
    Alpha=1
    ConstraintWeights_0=0.6806
    ConstraintWeights_1:float=0.0187
    ConstraintWeights_2:float=0.0027
    ConstraintWeights_3:float=0.0414
    ConstraintWeights_4:float=0.1111
    ConstraintWeights_5:float=0.1163
    ConstraintWeights_6:float=0.0292
   ) -> (Pose) {
    alphaBoolBlend {blendOption=Linear}
    boneToModify {boneName=dyn_shldr_L_low}
    constraintSetup [
     {targetBone {boneName=clavicle_l}}
     {targetBone {boneName=upperarm_bk_volume_l}}
     {targetBone {boneName=spine_04}}
     {targetBone {boneName=spine_05}}
     {targetBone {boneName=upperarm_twist_01_l}}
     {targetBone {boneName=upperarm_up_volume_l}}
     {targetBone {boneName=upperarm_fr_volume_l}}
    ]
    constraintWeights [1,1,1,1,1,1,1]
   }
   n27 = "Constraint - dyn_shldr_R_MSTR (PPPPPP)"!(
    ComponentPose=@n26
    Alpha=1
    ConstraintWeights_0=0.0198
    ConstraintWeights_1:float=0.0035
    ConstraintWeights_2:float=0.021
    ConstraintWeights_3:float=0.0051
    ConstraintWeights_4:float=0.1568
    ConstraintWeights_5:float=0.7938
   ) -> (Pose) {
    alphaBoolBlend {blendOption=Linear}
    boneToModify {boneName=dyn_shldr_R_MSTR}
    constraintSetup [
     {targetBone {boneName=upperarm_up_volume_r}}
     {targetBone {boneName=upperarm_fr_volume_r}}
     {targetBone {boneName=upperarm_twist_01_r}}
     {targetBone {boneName=spine_04}}
     {targetBone {boneName=spine_05}}
     {targetBone {boneName=clavicle_r}}
    ]
    constraintWeights [1,1,1,1,1,1]
   }
   n28 = "Constraint - dyn_shldr_R_high (PPPPPPPP)"!(
    ComponentPose=@n27
    Alpha=1
    ConstraintWeights_0=0
    ConstraintWeights_1:float=0
    ConstraintWeights_2:float=0
    ConstraintWeights_3:float=0.0048
    ConstraintWeights_4:float=0.3252
    ConstraintWeights_5:float=0.6699
    ConstraintWeights_6:float=0
    ConstraintWeights_7:float=0
   ) -> (Pose) {
    alphaBoolBlend {blendOption=Linear}
    boneToModify {boneName=dyn_shldr_R_high}
    constraintSetup [
     {targetBone {boneName=upperarm_up_volume_r}}
     {targetBone {boneName=upperarm_twist_01_r}}
     {targetBone {boneName=spine_03}}
     {targetBone {boneName=spine_04}}
     {targetBone {boneName=spine_05}}
     {targetBone {boneName=clavicle_r}}
     {targetBone {boneName=neck_02}}
     {targetBone {boneName=neck_01}}
    ]
    constraintWeights [1,1,1,1,1,1,1,1]
   }
   n29 = "Constraint - dyn_shldr_R_low (PPPPPPP)"!(
    ComponentPose=@n28
    Alpha=1
    ConstraintWeights_0=0.1163
    ConstraintWeights_1:float=0.0292
    ConstraintWeights_2:float=0.1111
    ConstraintWeights_3:float=0.0187
    ConstraintWeights_4:float=0.0027
    ConstraintWeights_5:float=0.0414
    ConstraintWeights_6:float=0.6806
   ) -> (Pose) {
    alphaBoolBlend {blendOption=Linear}
    boneToModify {boneName=dyn_shldr_R_low}
    constraintSetup [
     {targetBone {boneName=upperarm_up_volume_r}}
     {targetBone {boneName=upperarm_fr_volume_r}}
     {targetBone {boneName=upperarm_twist_01_r}}
     {targetBone {boneName=upperarm_bk_volume_r}}
     {targetBone {boneName=spine_04}}
     {targetBone {boneName=spine_05}}
     {targetBone {boneName=clavicle_r}}
    ]
    constraintWeights [1,1,1,1,1,1,1]
   }
   Save cached pose 'Pre-sim'(Pose=Component To Local(ComponentPose=@n29)) {cacheName=Pre-sim}
  ]}
 EventGraph {nodes [
   Event BlueprintUpdateAnimation()
   > n1 = Sequence!(execute) -> (then_0, then_1, then_2:Exec, then_3:Exec, then_4:Exec, then_5:Exec, then_6:Exec)
   n2 = GetDeltaTransformFromRefPose(self=GetOwnerSkeletalMeshComponent(), BoneName=lowerarm_l, BaseName=upperarm_l)
   > Set elbow_rot_l(elbow_rot_l:double=BreakRotator(InRot=BreakTransform(InTransform=@n2).Rotation).Yaw) -> (Output_Get:double)
   n3 = GetDeltaTransformFromRefPose(self=GetOwnerSkeletalMeshComponent(), BoneName=lowerarm_r, BaseName=upperarm_r)
   > Set elbow_rot_r(elbow_rot_r:double=BreakRotator(InRot=BreakTransform(InTransform=@n3).Rotation).Yaw) -> (Output_Get:double)
   n4 = GetDeltaTransformFromRefPose(self=GetOwnerSkeletalMeshComponent(), BoneName=calf_l, BaseName=thigh_l)
   Set knee_rot_l(execute=@n1.then_2, knee_rot_l:double=BreakRotator(InRot=BreakTransform(InTransform=@n4).Rotation).Yaw) -> (Output_Get:double)
   n5 = GetDeltaTransformFromRefPose(self=GetOwnerSkeletalMeshComponent(), BoneName=calf_r, BaseName=thigh_r)
   > Set knee_rot_r(knee_rot_r:double=BreakRotator(InRot=BreakTransform(InTransform=@n5).Rotation).Yaw) -> (Output_Get:double)
   n6 = GetOwnerSkeletalMeshComponent()
   n7 = BreakRotator(InRot=BreakTransform(InTransform=GetDeltaTransformFromRefPose(self=@n6, BoneName=upperarm_l, BaseName=clavicle_l)).Rotation)
   Set UpperArm_L _rot(execute=@n1.then_3, UpperArm_L _rot:double=@n7.Pitch) -> (Output_Get:double)
   n8 = GetOwnerSkeletalMeshComponent()
   n9 = BreakRotator(InRot=BreakTransform(InTransform=GetDeltaTransformFromRefPose(self=@n8, BoneName=upperarm_r, BaseName=clavicle_r)).Rotation)
   > Set UpperArm_R_rot(UpperArm_R_rot:double=@n9.Pitch) -> (Output_Get:double)
   n10 = GetOwnerSkeletalMeshComponent()
   n11 = BreakRotator(InRot=BreakTransform(InTransform=GetDeltaTransformFromRefPose(self=@n10, BoneName=neck_02, BaseName=spine_05)).Rotation)
   Set Neck_Back_Rot(
    execute=@n1.then_4
    Neck_Back_Rot:double=SelectFloat(
     A=MapRangeClamped(Value=@n11.Yaw, InRangeA=10, InRangeB=-10)
     B=MapRangeClamped(Value=@n11.Yaw, InRangeA=-10, InRangeB=-30)
     bPickA=Get bIsTargeting()
    )
   ) -> (Output_Get:double)
   n12 = GetDeltaTransformFromRefPose(self=GetOwnerSkeletalMeshComponent(), BoneName=thigh_l, BaseName=pelvis)
   Set thigh_rot_l(execute=@n1.then_1, thigh_rot_l=BreakRotator(InRot=BreakTransform(InTransform=@n12).Rotation).Yaw)
   n13 = GetDeltaTransformFromRefPose(self=GetOwnerSkeletalMeshComponent(), BoneName=thigh_r, BaseName=pelvis)
   > Set thigh_rot_r(thigh_rot_r=BreakRotator(InRot=BreakTransform(InTransform=@n13).Rotation).Yaw)
   Event OnCharacterCustomizationLoaded()
   > Set BodyMesh(BodyMesh=GetOwnerSkeletalMeshComponent())
   > n14 = Cast To PlayerPawn_Athena(Object=TryGetPawnOwner())
   > n15 = Set PawnOwner(PawnOwner:Player Pawn Athena Object Reference=@n14) -> (Output_Get:Player Pawn Athena Object Reference)
   > Set TailMesh(TailMesh:Skeletal Mesh Component Object Reference=GetSkeletalMeshForPartType(self=@n15, PartType=MiscOrTail)) -> (Output_Get:Skeletal Mesh Component Object Reference)
   n16 = CheckForBackbling()
   > Set isWearingBackbling(isWearingBackbling=@n16)
   > Set isWearingCape(isWearingCape=@n16.bIsBackblingLong)
   > Set characterHasTail(characterHasTail:bool=CheckForTail!(self:Custom Character Part Anim Instance Object Reference) -> (bHasTail:bool)) -> (Output_Get:bool)
   > Branch(Condition=Get characterHasTail() -> (characterHasTail:bool))
   > n17 = Branch(Condition=Get isWearingCape() -> (isWearingCape:bool))
   n18 = Get TailMesh() -> (TailMesh:Skeletal Mesh Component Object Reference)
   > SetVisibility!(execute, self:Scene Component Object Reference=@n18, bNewVisibility:bool=false, bPropagateToChildren:bool=false) -> (then)
   > Set cape and tail combo(cape and tail combo:bool=false) -> (Output_Get:bool)
   SetVisibility(execute=@n17.else, self=@n18, bNewVisibility=true, bPropagateToChildren=false)
   > n19 = Branch(Condition=Get bIsInFrontEnd())
   > Set Cape and Tail Combo Frontend(Cape and Tail Combo Frontend:bool=true) -> (Output_Get:bool)
   Set cape and tail combo(execute=@n19.else, cape and tail combo=true)
   TryGetPawnOwner()
   "Absolute (Float)"(A=@n7.Pitch)
   "Absolute (Float)"(A=@n9.Pitch)
  ]}
]
variables [
 Arms {
  type=PoseDeformationQuery_Arms
  default {
   armBackL {angle=50, bDrawDebug=false, endJoint=lowerarm_l, poseJoint=pose_arm_bk_l, startJoint=upperarm_l}
   armBackR {angle=50, bDrawDebug=false, endJoint=lowerarm_r, poseJoint=pose_arm_bk_r, startJoint=upperarm_r}
   armForwardDownL {angle=40, bDrawDebug=false, endJoint=lowerarm_l, poseJoint=pose_arm_fwd_dn_l, startJoint=upperarm_l}
   armForwardDownR {angle=40, bDrawDebug=false, endJoint=lowerarm_r, poseJoint=pose_arm_fwd_dn_r, startJoint=upperarm_r}
   armForwardL {angle=50, bDrawDebug=false, endJoint=lowerarm_l, poseJoint=pose_arm_fwd_l, startJoint=upperarm_l}
   armForwardR {angle=50, bDrawDebug=false, endJoint=lowerarm_r, poseJoint=pose_arm_fwd_r, startJoint=upperarm_r}
   armForwardUpL {angle=45, bDrawDebug=false, endJoint=lowerarm_l, poseJoint=pose_arm_fwd_up_l, startJoint=upperarm_l}
   armForwardUpR {angle=45, bDrawDebug=false, endJoint=lowerarm_r, poseJoint=pose_arm_fwd_up_r, startJoint=upperarm_r}
   armUpL {angle=70, bDrawDebug=false, endJoint=lowerarm_l, poseJoint=pose_arm_up_l, startJoint=upperarm_l}
   armUpR {angle=70, bDrawDebug=false, endJoint=lowerarm_r, poseJoint=pose_arm_up_r, startJoint=upperarm_r}
  }
 }
 BodyMesh {type=SkeletalMeshComponent}
 cape and tail combo {type=boolean}
 Cape and Tail Combo Frontend {type=boolean}
 characterHasTail {type=boolean}
 elbow_rot_l {type=number}
 elbow_rot_r {type=number}
 isWearingBackbling {type=boolean}
 isWearingCape {type=boolean}
 knee_rot_l {type=number}
 knee_rot_r {type=number}
 Neck_Back_Rot {type=number}
 PawnOwner {type=PlayerPawn_Athena_C}
 TailMesh {type=SkeletalMeshComponent}
 thigh_rot_l {type=number}
 thigh_rot_r {type=number}
 UpperArm_L _rot {type=number}
 UpperArm_R_rot {type=number}
]
