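# Physics asset for the face accessory mesh of the female medium Bistro Monster cosmetic body, covering its dynamic hair and tail bones along with the main body.
# A six-segment hair capsule chain hangs from the head through a small base body, with tight twist and moderate swing limits between segments.
# A four-segment tail chain hangs from a pelvis socket.
# Head, torso, arm and thigh capsules form the body shape, joined like the tail with free angular constraints.
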PhysicsAsset /BRCosmetics/Characters/Player/Female/Medium/Bodies/F_MED_Bistro_Monster/Meshes/Parts/F_MED_Bistro_Monster_FaceAcc_FE_Physics
bodies {
 dyn_hair_a [{shapeType=Capsule, length=8, radius=5.5, center=<10 0 0>, rotation=<rot 90 0 0>}]
 dyn_hair_a_base [{shapeType=Capsule, length=1, radius=1}]
 dyn_hair_b [{shapeType=Capsule, length=7.75, radius=5.25, center=<9 0 0>, rotation=<rot 90 0 0>}]
 dyn_hair_c [{shapeType=Capsule, length=9.75, radius=5.5, center=<11 0 0>, rotation=<rot 90 0 0>}]
 dyn_hair_d [{shapeType=Capsule, length=8.3125, radius=5.9375, center=<10 0 1.2262>, rotation=<rot 90 0 0>}]
 dyn_hair_e [{shapeType=Capsule, length=14.5625, radius=4.6875, center=<12 0 0>, rotation=<rot 90 0 0>}]
 dyn_hair_f [{shapeType=Capsule, length=11.25, radius=3.75, center=<10 0 0>, rotation=<rot 90 0 0>}]
 dyn_tail_socket [{shapeType=Capsule, length=0.505, radius=0.505, shapeName=dyn_tail_socket_capsule, rotation=<rot 0 -90 0>}]
 head [{shapeType=Capsule, length=11.1311, radius=11.1554, shapeName=head_capsule, center=<2.3501 0.5252 0.0596>, rotation=<rot 90 0 0>}]
 lowerarm_l [{shapeType=Capsule, length=28, radius=4.75, center=<15.7336 -0.0001 0>, rotation=<rot 90 0 0>}]
 lowerarm_r [{shapeType=Capsule, length=28, radius=4.75, center=<-15.7336 0.0001 0>, rotation=<rot 90 0 180>}]
 pelvis [{shapeType=Capsule, length=9.25, radius=16, center=<-3.742 0.8329 0>}]
 spine_05 [{shapeType=Capsule, length=12.25, radius=12.5, center=<-0.8242 3.3872 -0.0002>}]
 tail_1 [{shapeType=Capsule, length=11.1875, radius=4.8125, shapeName=tail_1_capsule, center=<9.5897 0.0001 0>, rotation=<rot 90 0 0>}]
 tail_2 [{shapeType=Capsule, length=11.1875, radius=4.8125, shapeName=tail_2_capsule, center=<9.6734 0.0001 0>, rotation=<rot 90 0 0>}]
 tail_3 [{shapeType=Capsule, length=18.0625, radius=4.8125, shapeName=tail_3_capsule, center=<12.7461 0 0>, rotation=<rot 90 0 0>}]
 tail_base [{shapeType=Capsule, length=11.1875, radius=4.8125, shapeName=tail_base_capsule, center=<8.5644 0.0001 0>, rotation=<rot 90 0 0>}]
 thigh_l [{shapeType=Capsule, length=25, radius=14.75, center=<-24.665 0 -0.0002>, rotation=<rot 90 0 0>}]
 thigh_r [{shapeType=Capsule, length=25, radius=14.75, center=<24.665 0 0.0002>, rotation=<rot 90 0 180>}]
 upperarm_l [{shapeType=Capsule, length=20.5, radius=5, center=<12.954 -0.0001 0>, rotation=<rot 90 0 0>}]
 upperarm_r [{shapeType=Capsule, length=20.5, radius=5, center=<-12.954 0.0001 0>, rotation=<rot 90 0 180>}]
}
constraints [
 {bone1Name=dyn_hair_a, bone2Name=dyn_hair_a_base, swing1LimitDegrees=30, swing2LimitDegrees=30, twistLimitDegrees=5}
 {bone1Name=dyn_hair_b, bone2Name=dyn_hair_a, swing1LimitDegrees=30, swing2LimitDegrees=30, twistLimitDegrees=5}
 {bone1Name=dyn_hair_c, bone2Name=dyn_hair_b, swing1LimitDegrees=30, swing2LimitDegrees=30, twistLimitDegrees=5}
 {bone1Name=dyn_hair_d, bone2Name=dyn_hair_c, swing1LimitDegrees=30, swing2LimitDegrees=30, twistLimitDegrees=5}
 {bone1Name=dyn_hair_e, bone2Name=dyn_hair_d, swing1LimitDegrees=30, swing2LimitDegrees=30, twistLimitDegrees=5}
 {bone1Name=dyn_hair_f, bone2Name=dyn_hair_e, swing1LimitDegrees=30, swing2LimitDegrees=30, twistLimitDegrees=5}
 {
  bone1Name=dyn_hair_a_base
  bone2Name=head
  swing1LimitDegrees=45
  swing1Motion=Free
  swing2LimitDegrees=45
  swing2Motion=Free
  twistLimitDegrees=45
  twistMotion=Free
 }
 {
  bone1Name=spine_05
  bone2Name=pelvis
  swing1LimitDegrees=45
  swing1Motion=Free
  swing2LimitDegrees=45
  swing2Motion=Free
  twistLimitDegrees=45
  twistMotion=Free
 }
 {
  bone1Name=upperarm_l
  bone2Name=spine_05
  swing1LimitDegrees=45
  swing1Motion=Free
  swing2LimitDegrees=45
  swing2Motion=Free
  twistLimitDegrees=45
  twistMotion=Free
 }
 {
  bone1Name=lowerarm_l
  bone2Name=upperarm_l
  swing1LimitDegrees=45
  swing1Motion=Free
  swing2LimitDegrees=45
  swing2Motion=Free
  twistLimitDegrees=45
  twistMotion=Free
 }
 {
  bone1Name=upperarm_r
  bone2Name=spine_05
  swing1LimitDegrees=45
  swing1Motion=Free
  swing2LimitDegrees=45
  swing2Motion=Free
  twistLimitDegrees=45
  twistMotion=Free
 }
 {
  bone1Name=lowerarm_r
  bone2Name=upperarm_r
  swing1LimitDegrees=45
  swing1Motion=Free
  swing2LimitDegrees=45
  swing2Motion=Free
  twistLimitDegrees=45
  twistMotion=Free
 }
 {
  bone1Name=thigh_l
  bone2Name=pelvis
  swing1LimitDegrees=45
  swing1Motion=Free
  swing2LimitDegrees=45
  swing2Motion=Free
  twistLimitDegrees=45
  twistMotion=Free
 }
 {
  bone1Name=thigh_r
  bone2Name=pelvis
  swing1LimitDegrees=45
  swing1Motion=Free
  swing2LimitDegrees=45
  swing2Motion=Free
  twistLimitDegrees=45
  twistMotion=Free
 }
 {
  bone1Name=dyn_tail_socket
  bone2Name=pelvis
  swing1LimitDegrees=45
  swing1Motion=Free
  swing2LimitDegrees=45
  swing2Motion=Free
  twistLimitDegrees=45
  twistMotion=Free
 }
 {
  bone1Name=tail_base
  bone2Name=dyn_tail_socket
  swing1LimitDegrees=45
  swing1Motion=Free
  swing2LimitDegrees=45
  swing2Motion=Free
  twistLimitDegrees=45
  twistMotion=Free
 }
 {
  bone1Name=tail_1
  bone2Name=tail_base
  swing1LimitDegrees=45
  swing1Motion=Free
  swing2LimitDegrees=45
  swing2Motion=Free
  twistLimitDegrees=45
  twistMotion=Free
 }
 {
  bone1Name=tail_2
  bone2Name=tail_1
  swing1LimitDegrees=45
  swing1Motion=Free
  swing2LimitDegrees=45
  swing2Motion=Free
  twistLimitDegrees=45
  twistMotion=Free
 }
 {
  bone1Name=tail_3
  bone2Name=tail_2
  swing1LimitDegrees=45
  swing1Motion=Free
  swing2LimitDegrees=45
  swing2Motion=Free
  twistLimitDegrees=45
  twistMotion=Free
 }
 {
  bone1Name=head
  bone2Name=spine_05
  swing1LimitDegrees=45
  swing1Motion=Free
  swing2LimitDegrees=45
  swing2Motion=Free
  twistLimitDegrees=45
  twistMotion=Free
 }
]
