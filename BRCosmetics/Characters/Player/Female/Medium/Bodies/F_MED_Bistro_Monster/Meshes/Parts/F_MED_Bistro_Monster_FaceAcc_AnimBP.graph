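# Post-process style AnimBP for the Bistro Monster female character's FaceAcc part, adding secondary motion and pose fixes to its hair and tail bones.
# A simulated branch layers tail-aware RigidBody, AnimDynamics and additive corrections for backblings, skydiving and jumping.
# A non-simulated branch uses a Trail controller with head-tilt offsets.
# bRigidBodyEnabled picks the branch, then spine constraints, springs and bone transforms adapt hair to short backblings.
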AnimBlueprint /BRCosmetics/Characters/Player/Female/Medium/Bodies/F_MED_Bistro_Monster/Meshes/Parts/F_MED_Bistro_Monster_FaceAcc_AnimBP
target_skeleton=&/F_MED_Bistro_Monster_FaceAcc_Skeleton
graphs [
 AnimGraph {nodes [
   n1 = Get skyDiveFloatPoseFixAlpha() -> (skyDiveFloatPoseFixAlpha:double)
   n2 = "Transform (Modify) Bone - Bone: dyn_hair_a_base"(Alpha=@n1, Translation="0,0,0", Rotation="0,-45,0", Scale="1,1,1") {
    rotationMode=BMM_Additive
    rotationSpace=BCS_BoneSpace
    scaleMode=BMM_Additive
    scaleSpace=BCS_BoneSpace
    translationMode=BMM_Additive
    translationSpace=BCS_BoneSpace
    alphaBoolBlend {blendOption=Linear}
    alphaScaleBiasClamp {inRange {max=0, min=-30}, outRange {max=0, min=1}}
    boneToModify {boneName=dyn_hair_a_base}
    rotation {yaw=-45}
   }
   n3 = "Transform (Modify) Bone - Bone: dyn_hair_a_base"(ComponentPose=@n2, Alpha=@n1, Translation="0,0,0", Rotation="0,20,0", Scale="1,1,1") {
    rotationMode=BMM_Additive
    rotationSpace=BCS_BoneSpace
    scaleMode=BMM_Additive
    scaleSpace=BCS_BoneSpace
    translationMode=BMM_Additive
    translationSpace=BCS_BoneSpace
    alphaBoolBlend {blendOption=Linear}
    alphaScaleBiasClamp {inRange {max=30}}
    boneToModify {boneName=dyn_hair_a_base}
    rotation {yaw=20}
   }
   n4 = "Transform (Modify) Bone - Bone: tail_base"(ComponentPose=@n3, Alpha=1, Translation="0,0,0", Rotation=Get Tail_Base_Rot() -> (Tail_Base_Rot:Rotator), Scale="1,1,1") {rotationMode=BMM_Replace, rotationSpace=BCS_BoneSpace, alphaBoolBlend {blendOption=Linear}, boneToModify {boneName=tail_base}}
   n5 = "Transform (Modify) Bone - Bone: tail_1"(ComponentPose=@n4, Alpha=1, Translation="0,0,0", Rotation=Get Tail_1_Rot() -> (Tail_1_Rot:Rotator), Scale="1,1,1") {
    rotationMode=BMM_Replace
    rotationSpace=BCS_BoneSpace
    translationSpace=BCS_BoneSpace
    alphaBoolBlend {blendOption=Linear}
    boneToModify {boneName=tail_1}
   }
   n6 = "Transform (Modify) Bone - Bone: tail_2"(ComponentPose=@n5, Alpha=1, Translation="0,0,0", Rotation=Get Tail_2_Rot() -> (Tail_2_Rot:Rotator), Scale="1,1,1") {
    rotationMode=BMM_Replace
    rotationSpace=BCS_BoneSpace
    translationSpace=BCS_BoneSpace
    alphaBoolBlend {blendOption=Linear}
    boneToModify {boneName=tail_2}
   }
   n7 = "Transform (Modify) Bone - Bone: tail_3"(ComponentPose=@n6, Alpha=1, Translation="0,0,0", Rotation=Get Tail_3_Rot() -> (Tail_3_Rot:Rotator), Scale="1,1,1") {
    rotationMode=BMM_Replace
    rotationSpace=BCS_BoneSpace
    translationSpace=BCS_BoneSpace
    alphaBoolBlend {blendOption=Linear}
    boneToModify {boneName=tail_3}
   }
   Component To Local(ComponentPose=@n7)
   n8 = Get IsShortBackblingOn() -> (IsShortBackblingOn:bool)
   n9 = "Constraint - dyn_hair_a (PP)"!(ComponentPose, bAlphaBoolEnabled:bool=@n8, ConstraintWeights_0=0.7, ConstraintWeights_1:float=0.3) -> (Pose) {
    alphaInputType=Bool
    alphaBoolBlend {blendOption=Linear}
    boneToModify {boneName=dyn_hair_a}
    constraintSetup [{targetBone {boneName=head}}, {targetBone {boneName=spine_05}}]
    constraintWeights [1,1]
   }
   n10 = "Constraint - dyn_hair_b (PP)"!(ComponentPose=@n9, bAlphaBoolEnabled:bool=@n8, ConstraintWeights_0=0.6, ConstraintWeights_1:float=0.4) -> (Pose) {
    alphaInputType=Bool
    alphaBoolBlend {blendOption=Linear}
    boneToModify {boneName=dyn_hair_b}
    constraintSetup [{targetBone {boneName=head}}, {targetBone {boneName=spine_05}}]
    constraintWeights [1,1]
   }
   n11 = "Constraint - dyn_hair_c (PP)"!(ComponentPose=@n10, bAlphaBoolEnabled:bool=@n8, ConstraintWeights_0=0.7, ConstraintWeights_1:float=0.3) -> (Pose) {
    alphaInputType=Bool
    alphaBoolBlend {blendOption=Linear}
    boneToModify {boneName=dyn_hair_c}
    constraintSetup [{targetBone {boneName=spine_05}}, {targetBone {boneName=spine_04}}]
    constraintWeights [1,1]
   }
   n12 = "Constraint - dyn_hair_d (PP)"!(ComponentPose=@n11, bAlphaBoolEnabled:bool=@n8, ConstraintWeights_0=0.6, ConstraintWeights_1:float=0.4) -> (Pose) {
    alphaInputType=Bool
    alphaBoolBlend {blendOption=Linear}
    boneToModify {boneName=dyn_hair_d}
    constraintSetup [{targetBone {boneName=spine_05}}, {targetBone {boneName=spine_04}}]
    constraintWeights [1,1]
   }
   "Constraint - dyn_hair_e (PP)"!(ComponentPose=@n12, bAlphaBoolEnabled:bool=@n8, ConstraintWeights_0=0.6, ConstraintWeights_1:float=0.4) -> (Pose) {
    alphaInputType=Bool
    alphaBoolBlend {blendOption=Linear}
    boneToModify {boneName=dyn_hair_e}
    constraintSetup [{targetBone {boneName=spine_04}}, {targetBone {boneName=spine_03}}]
    constraintWeights [1,1]
   }
   n13 = Get IsShortBackblingOn()
   n14 = "Constraint - dyn_hair_a (PP)"(bAlphaBoolEnabled=@n13, ConstraintWeights_0=0.7, ConstraintWeights_1=0.3) {
    alphaInputType=Bool
    alphaBoolBlend {blendOption=Linear}
    boneToModify {boneName=dyn_hair_a}
    constraintSetup [{targetBone {boneName=head}}, {targetBone {boneName=spine_05}}]
    constraintWeights [1,1]
   }
   n15 = "Constraint - dyn_hair_b (PP)"(ComponentPose=@n14, bAlphaBoolEnabled=@n13, ConstraintWeights_0=0.6, ConstraintWeights_1=0.4) {
    alphaInputType=Bool
    alphaBoolBlend {blendOption=Linear}
    boneToModify {boneName=dyn_hair_b}
    constraintSetup [{targetBone {boneName=head}}, {targetBone {boneName=spine_05}}]
    constraintWeights [1,1]
   }
   n16 = "Constraint - dyn_hair_c (PP)"(ComponentPose=@n15, bAlphaBoolEnabled=@n13, ConstraintWeights_0=0.7, ConstraintWeights_1=0.3) {
    alphaInputType=Bool
    alphaBoolBlend {blendOption=Linear}
    boneToModify {boneName=dyn_hair_c}
    constraintSetup [{targetBone {boneName=spine_05}}, {targetBone {boneName=spine_04}}]
    constraintWeights [1,1]
   }
   n17 = "Constraint - dyn_hair_d (PP)"(ComponentPose=@n16, bAlphaBoolEnabled=@n13, ConstraintWeights_0=0.6, ConstraintWeights_1=0.4) {
    alphaInputType=Bool
    alphaBoolBlend {blendOption=Linear}
    boneToModify {boneName=dyn_hair_d}
    constraintSetup [{targetBone {boneName=spine_05}}, {targetBone {boneName=spine_04}}]
    constraintWeights [1,1]
   }
   "Constraint - dyn_hair_e (PP)"(ComponentPose=@n17, bAlphaBoolEnabled=@n13, ConstraintWeights_0=0.6, ConstraintWeights_1=0.4) {
    alphaInputType=Bool
    alphaBoolBlend {blendOption=Linear}
    boneToModify {boneName=dyn_hair_e}
    constraintSetup [{targetBone {boneName=spine_04}}, {targetBone {boneName=spine_03}}]
    constraintWeights [1,1]
   }
   n18 = Get backpackOnHeadUpDn() -> (backpackOnHeadUpDn:double)
   n19 = "Transform (Modify) Bone - Bone: dyn_hair_b"(Alpha=@n18, Translation="1,-4,0", Rotation="0,-10,0", Scale="1,1,1") {
    rotationMode=BMM_Additive
    rotationSpace=BCS_BoneSpace
    scaleMode=BMM_Additive
    scaleSpace=BCS_BoneSpace
    translationMode=BMM_Additive
    translationSpace=BCS_BoneSpace
    alphaBoolBlend {blendOption=Linear}
    alphaScaleBiasClamp {bMapRange=true, inRange {max=0, min=-25}, outRange {max=0, min=1}}
    boneToModify {boneName=dyn_hair_b}
    rotation {yaw=-10}
    translation=<1 -4>
   }
   "Transform (Modify) Bone - Bone: dyn_hair_small_b"(ComponentPose=@n19, Alpha=@n18, Translation="0,-1.5,0", Rotation="0,0,0", Scale="1,1,1") {
    rotationMode=BMM_Additive
    rotationSpace=BCS_BoneSpace
    scaleMode=BMM_Additive
    scaleSpace=BCS_BoneSpace
    translationMode=BMM_Additive
    translationSpace=BCS_BoneSpace
    alphaBoolBlend {blendOption=Linear}
    alphaScaleBiasClamp {bMapRange=true, inRange {max=0, min=-25}, outRange {max=0, min=1}}
    boneToModify {boneName=dyn_hair_small_b}
    translation {y=-1.5}
   }
   "Transform (Modify) Bone - Bone: dyn_hair_d"(Alpha=@n18, Translation="0,0,0", Rotation="0,-20,0", Scale="1,1,1") {
    rotationMode=BMM_Additive
    rotationSpace=BCS_BoneSpace
    scaleMode=BMM_Additive
    scaleSpace=BCS_BoneSpace
    translationMode=BMM_Additive
    translationSpace=BCS_BoneSpace
    alphaBoolBlend {blendOption=Linear}
    alphaScaleBiasClamp {bMapRange=true, inRange {max=0, min=-25}, outRange {max=0, min=1}}
    boneToModify {boneName=dyn_hair_d}
    rotation {yaw=-20}
   }
   "Transform (Modify) Bone - Bone: dyn_hair_e"(Alpha=@n18, Translation="0,0,0", Rotation="0,20,0", Scale="1,1,1") {
    rotationMode=BMM_Additive
    rotationSpace=BCS_BoneSpace
    scaleMode=BMM_Additive
    scaleSpace=BCS_BoneSpace
    translationMode=BMM_Additive
    translationSpace=BCS_BoneSpace
    alphaBoolBlend {blendOption=Linear}
    alphaScaleBiasClamp {bMapRange=true, inRange {max=0, min=-25}, outRange {max=0, min=1}}
    boneToModify {boneName=dyn_hair_e}
    rotation {yaw=20}
   }
   n20 = RigidBody!(
    ComponentPose=Local To Component(LocalPose=Use cached pose 'SimCachPose'())
    Alpha=0
    OverrideWorldGravity:Vector=Get Gravity Override()
    ExternalForce
   ) -> (Pose) {
    bClampLinearTranslationLimitToRefPose=true
    bOverrideWorldGravity=true
    overridePhysicsAsset=&/F_MED_Bistro_Monster_FaceAcc_Physics
    simulationSpace=BaseBoneSpace
    alphaBoolBlend {blendOption=Linear}
    alphaScaleBiasClamp {outRange {max=0.4, min=1}}
    baseBoneRef {boneName=head}
    componentAppliedLinearAccClamp=<3000 3000 3000>
    componentLinearAccScale=<0.4 0.35 0.4>
    componentLinearVelScale=<0.4 0.35 0.4>
    overrideWorldGravity {z=-980}
   }
   n21 = RigidBody(ComponentPose=Local To Component(LocalPose=Use cached pose 'SimCachPose'()), Alpha=0, OverrideWorldGravity=Get Gravity Override()) {
    bClampLinearTranslationLimitToRefPose=true
    bOverrideWorldGravity=true
    overridePhysicsAsset=&/F_MED_Bistro_Monster_FaceAcc_FE_Physics
    simulationSpace=BaseBoneSpace
    alphaBoolBlend {blendOption=Linear}
    alphaScaleBiasClamp {outRange {max=0.4, min=1}}
    baseBoneRef {boneName=head}
    componentAppliedLinearAccClamp=<3000 3000 3000>
    componentLinearAccScale=<0.4 0.35 0.4>
    componentLinearVelScale=<0.4 0.35 0.4>
    overrideWorldGravity {z=-980}
   }
   n22 = Blend Poses by bool(BlendPose_0=Component To Local(ComponentPose=@n21), BlendPose_1=Component To Local(ComponentPose=@n20), bActiveValue=Get bIsInFrontEnd()) {blendType=Linear}
   n23 = Use cached pose 'NoSimCachePose'()
   n24 = Get bRigidBodyEnabled()
   n25 = Get IsShortBackblingOn()
   n26 = "Constraint - dyn_hair_a (PP)"(
    ComponentPose=Local To Component(LocalPose=Blend Poses by bool(BlendPose_0=@n22, BlendPose_1=@n23, bActiveValue=@n24) {blendType=Linear})
    bAlphaBoolEnabled=@n25
    ConstraintWeights_0=0.7
    ConstraintWeights_1=0.3
   ) {
    alphaInputType=Bool
    alphaBoolBlend {blendOption=Linear}
    boneToModify {boneName=dyn_hair_a}
    constraintSetup [{targetBone {boneName=head}}, {targetBone {boneName=spine_05}}]
    constraintWeights [1,1]
   }
   n27 = "Constraint - dyn_hair_b (PP)"(ComponentPose=@n26, bAlphaBoolEnabled=@n25, ConstraintWeights_0=0.6, ConstraintWeights_1=0.4) {
    alphaInputType=Bool
    alphaBoolBlend {blendOption=Linear}
    boneToModify {boneName=dyn_hair_b}
    constraintSetup [{targetBone {boneName=head}}, {targetBone {boneName=spine_05}}]
    constraintWeights [1,1]
   }
   n28 = "Constraint - dyn_hair_c (PP)"(ComponentPose=@n27, bAlphaBoolEnabled=@n25, ConstraintWeights_0=0.7, ConstraintWeights_1=0.3) {
    alphaInputType=Bool
    alphaBoolBlend {blendOption=Linear}
    boneToModify {boneName=dyn_hair_c}
    constraintSetup [{targetBone {boneName=spine_05}}, {targetBone {boneName=spine_04}}]
    constraintWeights [1,1]
   }
   n29 = "Constraint - dyn_hair_d (PP)"(ComponentPose=@n28, bAlphaBoolEnabled=@n25, ConstraintWeights_0=0.6, ConstraintWeights_1=0.4) {
    alphaInputType=Bool
    alphaBoolBlend {blendOption=Linear}
    boneToModify {boneName=dyn_hair_d}
    constraintSetup [{targetBone {boneName=spine_05}}, {targetBone {boneName=spine_04}}]
    constraintWeights [1,1]
   }
   n30 = "Constraint - dyn_hair_e (PP)"(ComponentPose=@n29, bAlphaBoolEnabled=@n25, ConstraintWeights_0=0.6, ConstraintWeights_1=0.4) {
    alphaInputType=Bool
    alphaBoolBlend {blendOption=Linear}
    boneToModify {boneName=dyn_hair_e}
    constraintSetup [{targetBone {boneName=spine_04}}, {targetBone {boneName=spine_03}}]
    constraintWeights [1,1]
   }
   n31 = SelectFloat(A=0.5, B=0, bPickA=Get IsShortBackblingOn())
   n32 = Anim Dynamics - Bone: dyn_hair_c(ComponentPose=@n30, Alpha=@n31, ExternalForce="0,0,0") {
    angularDampingOverride=0.9999
    angularSpringConstant=75
    bAngularSpring=true
    bGravityOverrideInSimSpace=true
    bOverrideAngularDamping=true
    simulationSpace=RootRelative
    alphaBoolBlend {blendOption=Linear}
    boundBone {boneName=dyn_hair_c}
    componentLinearAccScale=<1 1 1>
    componentLinearVelScale=<1 1 1>
    physicsBodyDefinitions [{
      sphereCollisionRadius=0
      boundBone {boneName=dyn_hair_c}
      boxExtents=<20 20 10>
      constraintSetup {angularLimitsMax=<10 25 10>, angularLimitsMin=<-10 -25 0>, angularTarget=<0 0 0>}
      localJointOffset=<2 0 0>
     }]
   }
   n33 = Anim Dynamics - Bone: dyn_hair_d(ComponentPose=@n32, Alpha=@n31, ExternalForce="0,0,0") {
    angularDampingOverride=0.9999
    angularSpringConstant=75
    bAngularSpring=true
    bGravityOverrideInSimSpace=true
    bOverrideAngularDamping=true
    simulationSpace=RootRelative
    alphaBoolBlend {blendOption=Linear}
    boundBone {boneName=dyn_hair_d}
    componentLinearAccScale=<1 1 1>
    componentLinearVelScale=<1 1 1>
    physicsBodyDefinitions [{
      sphereCollisionRadius=0
      boundBone {boneName=dyn_hair_d}
      boxExtents=<20 20 10>
      constraintSetup {angularLimitsMax=<10 25 10>, angularLimitsMin=<-10 -25 0>, angularTarget=<0 0 0>}
      localJointOffset=<2 0 0>
     }]
   }
   n34 = Anim Dynamics - Bone: dyn_hair_e(ComponentPose=@n33, Alpha=@n31, ExternalForce="0,0,0") {
    angularDampingOverride=0.9999
    angularSpringConstant=75
    bAngularSpring=true
    bGravityOverrideInSimSpace=true
    bOverrideAngularDamping=true
    simulationSpace=RootRelative
    alphaBoolBlend {blendOption=Linear}
    boundBone {boneName=dyn_hair_e}
    componentLinearAccScale=<1 1 1>
    componentLinearVelScale=<1 1 1>
    physicsBodyDefinitions [{
      sphereCollisionRadius=0
      boundBone {boneName=dyn_hair_e}
      boxExtents=<20 20 10>
      constraintSetup {angularLimitsMax=<10 25 10>, angularLimitsMin=<-10 -25 0>, angularTarget=<0 0 0>}
      localJointOffset=<2 0 0>
     }]
   }
   n35 = Anim Dynamics - Bone: dyn_hair_f(ComponentPose=@n34, Alpha=@n31, ExternalForce="0,0,0") {
    angularDampingOverride=0.9999
    angularSpringConstant=75
    bAngularSpring=true
    bGravityOverrideInSimSpace=true
    bOverrideAngularDamping=true
    simulationSpace=RootRelative
    alphaBoolBlend {blendOption=Linear}
    boundBone {boneName=dyn_hair_f}
    componentLinearAccScale=<1 1 1>
    componentLinearVelScale=<1 1 1>
    physicsBodyDefinitions [{
      sphereCollisionRadius=0
      boundBone {boneName=dyn_hair_f}
      boxExtents=<20 20 10>
      constraintSetup {angularLimitsMax=<10 25 10>, angularLimitsMin=<-10 -25 0>, angularTarget=<0 0 0>}
      localJointOffset=<2 0 0>
     }]
   }
   n36 = Get IsShortBackblingOn()
   n37 = "Transform (Modify) Bone - Bone: dyn_hair_a"!(ComponentPose=@n35, bAlphaBoolEnabled:bool=@n36, Translation="0,0,0", Rotation="0,-6,0", Scale="0.8,1,1") -> (Pose) {
    alphaInputType=Bool
    rotationMode=BMM_Additive
    rotationSpace=BCS_BoneSpace
    scaleMode=BMM_Additive
    scaleSpace=BCS_BoneSpace
    translationMode=BMM_Additive
    translationSpace=BCS_BoneSpace
    alphaBoolBlend {blendOption=Linear}
    alphaScaleBiasClamp {bMapRange=true, inRange {max=0, min=-25}, outRange {max=0, min=1}}
    boneToModify {boneName=dyn_hair_a}
    rotation {yaw=-6}
    scale {x=0.8}
   }
   n38 = "Transform (Modify) Bone - Bone: dyn_hair_small_b"!(ComponentPose=@n37, bAlphaBoolEnabled:bool=@n36, Translation="0,-4,0", Rotation="0,5,0", Scale="0.8,1,1") -> (Pose) {
    alphaInputType=Bool
    rotationMode=BMM_Additive
    rotationSpace=BCS_BoneSpace
    scaleMode=BMM_Additive
    scaleSpace=BCS_BoneSpace
    translationMode=BMM_Additive
    translationSpace=BCS_BoneSpace
    alphaBoolBlend {blendOption=Linear}
    alphaScaleBiasClamp {bMapRange=true, inRange {max=0, min=-25}, outRange {max=0, min=1}}
    boneToModify {boneName=dyn_hair_small_b}
    rotation {yaw=5}
    scale {x=0.8}
    translation {y=-4}
   }
   n39 = "Transform (Modify) Bone - Bone: dyn_hair_small_c"!(ComponentPose=@n38, bAlphaBoolEnabled:bool=@n36, Translation="0,-3,0", Rotation="0,0,0", Scale="0.9,0.9,0.9") -> (Pose) {
    alphaInputType=Bool
    rotationMode=BMM_Additive
    rotationSpace=BCS_BoneSpace
    scaleMode=BMM_Additive
    scaleSpace=BCS_BoneSpace
    translationMode=BMM_Additive
    translationSpace=BCS_BoneSpace
    alphaBoolBlend {blendOption=Linear}
    alphaScaleBiasClamp {bMapRange=true, inRange {max=0, min=-25}, outRange {max=0, min=1}}
    boneToModify {boneName=dyn_hair_small_c}
    scale=<0.9 0.9 0.9>
    translation {y=-3}
   }
   n40 = "Transform (Modify) Bone - Bone: dyn_hair_c"!(ComponentPose=@n39, bAlphaBoolEnabled:bool=@n36, Translation="-1.5,0,5", Rotation="-10,0,0", Scale="0.85,0.85,0.85") -> (Pose) {
    alphaInputType=Bool
    rotationMode=BMM_Additive
    rotationSpace=BCS_BoneSpace
    scaleMode=BMM_Additive
    scaleSpace=BCS_BoneSpace
    translationMode=BMM_Additive
    translationSpace=BCS_BoneSpace
    alphaBoolBlend {blendOption=Linear}
    alphaScaleBiasClamp {bMapRange=true, inRange {max=0, min=-25}, outRange {max=0, min=1}}
    boneToModify {boneName=dyn_hair_c}
    rotation {pitch=-10}
    scale=<0.85 0.85 0.85>
    translation {x=-1.5, z=5}
   }
   n41 = "Transform (Modify) Bone - Bone: dyn_hair_d"!(ComponentPose=@n40, bAlphaBoolEnabled:bool=@n36, Translation="-1.5,-2,0", Rotation="0,-5,0", Scale="0.8,0.8,0.8") -> (Pose) {
    alphaInputType=Bool
    rotationMode=BMM_Additive
    rotationSpace=BCS_BoneSpace
    scaleMode=BMM_Additive
    scaleSpace=BCS_BoneSpace
    translationMode=BMM_Additive
    translationSpace=BCS_BoneSpace
    alphaBoolBlend {blendOption=Linear}
    alphaScaleBiasClamp {bMapRange=true, inRange {max=0, min=-25}, outRange {max=0, min=1}}
    boneToModify {boneName=dyn_hair_d}
    rotation {yaw=-5}
    scale=<0.8 0.8 0.8>
    translation=<-1.5 -2>
   }
   n42 = "Transform (Modify) Bone - Bone: dyn_hair_e"!(ComponentPose=@n41, bAlphaBoolEnabled:bool=@n36, Translation="-1.5,0,0", Rotation="0,0,0", Scale="0.6,0.6,0.6") -> (Pose) {
    alphaInputType=Bool
    rotationMode=BMM_Additive
    rotationSpace=BCS_BoneSpace
    scaleMode=BMM_Additive
    scaleSpace=BCS_BoneSpace
    translationMode=BMM_Additive
    translationSpace=BCS_BoneSpace
    alphaBoolBlend {blendOption=Linear}
    alphaScaleBiasClamp {bMapRange=true, inRange {max=0, min=-25}, outRange {max=0, min=1}}
    boneToModify {boneName=dyn_hair_e}
    scale=<0.6 0.6 0.6>
    translation {x=-1.5}
   }
   n43 = "Transform (Modify) Bone - Bone: dyn_hair_c"!(ComponentPose=@n42, Alpha:float=Get backpackOnHeadUpDn(), Translation="0,0,0", Rotation="0,-6,0", Scale="1,1,1") -> (Pose) {
    rotationMode=BMM_Additive
    rotationSpace=BCS_BoneSpace
    scaleMode=BMM_Additive
    scaleSpace=BCS_BoneSpace
    translationMode=BMM_Additive
    translationSpace=BCS_BoneSpace
    alphaBoolBlend {blendOption=Linear}
    alphaScaleBiasClamp {bMapRange=true, inRange {max=0, min=-25}, outRange {max=0, min=1}}
    boneToModify {boneName=dyn_hair_c}
    rotation {yaw=-6}
   }
   Output Pose(Result=Component To Local(ComponentPose=@n43)) {layerGroup=None}
   n44 = Trail controller - Bone: dyn_hair_f!(
    ComponentPose=Local To Component(LocalPose=Use cached pose 'SharedCachePose'())
    Alpha=SelectFloat(A=0, B=Get TrailAlpha() -> (TrailAlpha:double), bPickA=Get IsShortBackblingOn())
    RelaxationSpeedScale:float=Get RelaxSpeed() -> (RelaxSpeed:double)
   ) -> (Pose) {
    bLimitRotation=true
    chainLength=7
    maxDeltaTime=0.0333
    alphaScaleBiasClamp {bInterpResult=true}
    rotationLimits [
     {limitMax=<180 180 180>}
     {limitMax=<180 180 180>}
     {limitMax=<180 180 180>}
     {limitMax=<180 180 180>}
     {limitMax=<180 180 180>}
     {limitMax=<180 180 180>}
     {limitMax=<180 180 180>}
    ]
    rotationOffsets [<0 0 0>, <0 0 0>, <0 0 0>, <0 0 0>, <0 0 0>, <0 0 0>, <0 0 0>]
    trailBone {boneName=dyn_hair_f}
    trailRelaxationSpeed {editorCurveData {keys [<at 0 10 RCIM_Linear>,<at 0.5107 8.1304 RCIM_Linear>,<at 1 5 RCIM_Linear>]}}
   }
   n45 = "Transform (Modify) Bone - Bone: dyn_hair_a_base"(ComponentPose=@n44, Alpha=Get skyDiveFloatPoseFixAlpha(), Translation="0,0,0", Rotation="0,-45,0", Scale="1,1,1") {
    rotationMode=BMM_Additive
    rotationSpace=BCS_BoneSpace
    scaleMode=BMM_Additive
    scaleSpace=BCS_BoneSpace
    translationMode=BMM_Additive
    translationSpace=BCS_BoneSpace
    alphaBoolBlend {blendOption=Linear}
    alphaScaleBiasClamp {inRange {max=0, min=-30}, outRange {max=0, min=1}}
    boneToModify {boneName=dyn_hair_a_base}
    rotation {yaw=-45}
   }
   n46 = Get head_fwdBwd_rot() -> (head_fwdBwd_rot:double)
   n47 = "Transform (Modify) Bone - Bone: dyn_hair_a_base"(ComponentPose=@n45, Alpha=@n46, Translation="0,0,0", Rotation="0,-50,0", Scale="1,1,1") {
    rotationMode=BMM_Additive
    rotationSpace=BCS_BoneSpace
    scaleMode=BMM_Additive
    scaleSpace=BCS_BoneSpace
    translationMode=BMM_Additive
    translationSpace=BCS_BoneSpace
    alphaBoolBlend {blendOption=Linear}
    alphaScaleBiasClamp {bMapRange=true, inRange {max=0, min=-30}, outRange {max=0, min=1}}
    boneToModify {boneName=dyn_hair_a_base}
    rotation {yaw=-50}
   }
   n48 = "Transform (Modify) Bone - Bone: dyn_hair_a_base"(ComponentPose=@n47, Alpha=@n46, Translation="0,0,0", Rotation="0,10,0", Scale="1,1,1") {
    rotationMode=BMM_Additive
    rotationSpace=BCS_BoneSpace
    scaleMode=BMM_Additive
    scaleSpace=BCS_BoneSpace
    translationMode=BMM_Additive
    translationSpace=BCS_BoneSpace
    alphaBoolBlend {blendOption=Linear}
    alphaScaleBiasClamp {bMapRange=true, inRange {max=30}}
    boneToModify {boneName=dyn_hair_a_base}
    rotation {yaw=10}
   }
   Save cached pose 'NoSimCachePose'(Pose=Component To Local(ComponentPose=@n48)) {cacheName=NoSimCachePose}
   n49 = RigidBody(
    ComponentPose=Local To Component(LocalPose=Use cached pose 'SharedCachePose'())
    Alpha=Get rigidBodyAlpha() -> (rigidBodyAlpha:double)
    OverrideWorldGravity=Get Gravity Override()
   ) {
    bClampLinearTranslationLimitToRefPose=true
    bOverrideWorldGravity=true
    overridePhysicsAsset=&F_MED_Bistro_Monster_FaceAcc_Physics
    simulationSpace=BaseBoneSpace
    alphaBoolBlend {blendOption=Linear}
    alphaScaleBiasClamp {outRange {max=0.4, min=1}}
    baseBoneRef {boneName=head}
    componentAppliedLinearAccClamp=<3000 3000 3000>
    componentLinearAccScale=<0.4 0.35 0.4>
    componentLinearVelScale=<0.4 0.35 0.4>
    overrideWorldGravity {z=-980}
   }
   n50 = Component To Local(ComponentPose=@n49)
   n51 = RigidBody(
    ComponentPose=Local To Component(LocalPose=Use cached pose 'SharedCachePose'())
    Alpha=Get rigidBodyAlpha()
    OverrideWorldGravity=Get Gravity Override()
   ) {
    bClampLinearTranslationLimitToRefPose=true
    bOverrideWorldGravity=true
    overridePhysicsAsset=&/F_MED_Bistro_Monster_FaceAcc_NOTail_Physics
    simulationSpace=BaseBoneSpace
    alphaBoolBlend {blendOption=Linear}
    alphaScaleBiasClamp {outRange {max=0.4, min=1}}
    baseBoneRef {boneName=head}
    componentAppliedLinearAccClamp=<3000 3000 3000>
    componentLinearAccScale=<0.4 0.35 0.4>
    componentLinearVelScale=<0.4 0.35 0.4>
    overrideWorldGravity {z=-980}
   }
   n52 = Component To Local(ComponentPose=@n51)
   n53 = Get IsTailOn() -> (IsTailOn:bool)
   n54 = Get IsShortBackblingOnAlpha() -> (IsShortBackblingOnAlpha:double)
   n55 = Anim Dynamics - Bone: dyn_hair_small_a(
    ComponentPose=Local To Component(LocalPose=Blend Poses by bool(BlendPose_0=@n50, BlendPose_1=@n52, bActiveValue=@n53) {blendType=Linear})
    Alpha=@n54
    ExternalForce="0,0,0"
   ) {
    angularDampingOverride=1
    angularSpringConstant=20000
    bAngularSpring=true
    bGravityOverrideInSimSpace=true
    bOverrideAngularDamping=true
    bUseGravityOverride=true
    gravityScale=0.35
    simulationSpace=RootRelative
    alphaBoolBlend {blendOption=Linear}
    alphaScaleBiasClamp {bMapRange=true, outRange {max=0, min=0.5}}
    boundBone {boneName=dyn_hair_small_a}
    componentAppliedLinearAccClamp=<3000 3000 3000>
    componentLinearAccScale=<1 1 1>
    componentLinearVelScale=<1 1 1>
    gravityOverride {z=-980}
    physicsBodyDefinitions [{
      sphereCollisionRadius=0
      boundBone {boneName=dyn_hair_small_a}
      boxExtents=<20 30 30>
      constraintSetup {angularLimitsMax=<0 0 -10>, angularLimitsMin=<0 -30 -45>, angularTarget=<1 0 0>}
      localJointOffset=<7 0 0>
     }]
    sphericalLimits [
     {limitRadius=20, drivingBone {boneName=spine_03}, sphereLocalOffset=<0 0 0>}
     {limitRadius=20, drivingBone {boneName=spine_04}, sphereLocalOffset=<0 0 0>}
    ]
   }
   n56 = Anim Dynamics - Bone: dyn_hair_small_b(ComponentPose=@n55, Alpha=@n54, ExternalForce="0,0,0") {
    angularDampingOverride=1
    angularSpringConstant=20000
    bAngularSpring=true
    bGravityOverrideInSimSpace=true
    bOverrideAngularDamping=true
    bUseGravityOverride=true
    gravityScale=0.35
    simulationSpace=RootRelative
    alphaBoolBlend {blendOption=Linear}
    alphaScaleBiasClamp {bMapRange=true, outRange {max=0, min=0.9}}
    boundBone {boneName=dyn_hair_small_b}
    componentAppliedLinearAccClamp=<3000 3000 3000>
    componentLinearAccScale=<1 1 1>
    componentLinearVelScale=<1 1 1>
    gravityOverride {z=-980}
    physicsBodyDefinitions [{
      sphereCollisionRadius=0
      boundBone {boneName=dyn_hair_small_b}
      boxExtents=<20 30 30>
      constraintSetup {angularLimitsMax=<0 0 0>, angularLimitsMin=<0 -45 -25>, angularTarget=<1 0 0>}
      localJointOffset=<7 0 0>
     }]
    sphericalLimits [
     {limitRadius=20, drivingBone {boneName=spine_03}, sphereLocalOffset=<0 0 0>}
     {limitRadius=20, drivingBone {boneName=spine_04}, sphereLocalOffset=<0 0 0>}
    ]
   }
   n57 = Anim Dynamics - Bone: dyn_hair_small_c(ComponentPose=@n56, Alpha=@n54, ExternalForce="0,0,0") {
    angularDampingOverride=1
    angularSpringConstant=20000
    bAngularSpring=true
    bGravityOverrideInSimSpace=true
    bOverrideAngularDamping=true
    bUseGravityOverride=true
    gravityScale=0.35
    simulationSpace=RootRelative
    alphaBoolBlend {blendOption=Linear}
    alphaScaleBiasClamp {bMapRange=true, outRange {max=0, min=0.7}}
    boundBone {boneName=dyn_hair_small_c}
    componentAppliedLinearAccClamp=<3000 3000 3000>
    componentLinearAccScale=<1 1 1>
    componentLinearVelScale=<1 1 1>
    gravityOverride {z=-980}
    physicsBodyDefinitions [{
      sphereCollisionRadius=0
      boundBone {boneName=dyn_hair_small_c}
      boxExtents=<20 30 30>
      constraintSetup {angularLimitsMax=<15 0 0>, angularLimitsMin=<-15 -45 -45>, angularTarget=<1 0 0>}
      localJointOffset=<7 0 0>
     }]
    sphericalLimits [
     {limitRadius=20, drivingBone {boneName=spine_03}, sphereLocalOffset=<0 0 0>}
     {limitRadius=20, drivingBone {boneName=spine_04}, sphereLocalOffset=<0 0 0>}
    ]
   }
   n58 = Get IsShortBackblingOnAlpha()
   n59 = Anim Dynamics - Bone: dyn_hair_d(ComponentPose=@n57, Alpha=@n58, ExternalForce="0,0,0") {
    angularDampingOverride=1
    angularSpringConstant=20000
    bAngularSpring=true
    bGravityOverrideInSimSpace=true
    bOverrideAngularDamping=true
    bUseGravityOverride=true
    bUseSphericalLimits=true
    gravityScale=0.35
    simulationSpace=RootRelative
    alphaBoolBlend {blendOption=Linear}
    alphaScaleBiasClamp {bMapRange=true, outRange {max=0.5}}
    boundBone {boneName=dyn_hair_d}
    componentAppliedLinearAccClamp=<3000 3000 3000>
    componentLinearAccScale=<0.5 1 1>
    componentLinearVelScale=<0.5 1 1>
    gravityOverride {z=-980}
    physicsBodyDefinitions [{
      sphereCollisionRadius=0
      boundBone {boneName=dyn_hair_d}
      boxExtents=<20 30 40>
      constraintSetup {angularLimitsMax=<0 45 0>, angularLimitsMin=<0 -45 0>, angularTarget=<1 0 0>}
      localJointOffset=<10 0 0>
     }]
    sphericalLimits [
     {limitRadius=10, drivingBone {boneName=spine_03}, sphereLocalOffset=<0 0 0>}
     {limitRadius=10, drivingBone {boneName=spine_04}, sphereLocalOffset=<0 0 0>}
    ]
   }
   n60 = Anim Dynamics - Bone: dyn_hair_e(ComponentPose=@n59, Alpha=@n58, ExternalForce="0,0,0") {
    angularDampingOverride=1
    angularSpringConstant=20000
    bAngularSpring=true
    bGravityOverrideInSimSpace=true
    bOverrideAngularDamping=true
    bUseGravityOverride=true
    bUseSphericalLimits=true
    gravityScale=0.35
    simulationSpace=RootRelative
    alphaBoolBlend {blendOption=Linear}
    alphaScaleBiasClamp {bMapRange=true, outRange {max=0.5}}
    boundBone {boneName=dyn_hair_e}
    componentAppliedLinearAccClamp=<3000 3000 3000>
    componentLinearAccScale=<0.5 1 1>
    componentLinearVelScale=<0.5 1 1>
    gravityOverride {z=-980}
    physicsBodyDefinitions [{
      sphereCollisionRadius=0
      boundBone {boneName=dyn_hair_e}
      boxExtents=<20 40 40>
      constraintSetup {angularLimitsMax=<0 45 0>, angularLimitsMin=<0 -45 0>, angularTarget=<1 0 0>}
      localJointOffset=<10 0 0>
     }]
    sphericalLimits [
     {limitRadius=10, drivingBone {boneName=spine_03}, sphereLocalOffset=<0 0 0>}
     {limitRadius=10, drivingBone {boneName=spine_04}, sphereLocalOffset=<0 0 0>}
    ]
   }
   n61 = Anim Dynamics - Bone: dyn_hair_f(ComponentPose=@n60, Alpha=@n58, ExternalForce="0,0,0") {
    angularDampingOverride=1
    angularSpringConstant=20000
    bAngularSpring=true
    bGravityOverrideInSimSpace=true
    bOverrideAngularDamping=true
    bUseGravityOverride=true
    bUseSphericalLimits=true
    gravityScale=0.35
    simulationSpace=RootRelative
    alphaBoolBlend {blendOption=Linear}
    alphaScaleBiasClamp {bMapRange=true, outRange {max=0.5}}
    boundBone {boneName=dyn_hair_f}
    componentAppliedLinearAccClamp=<3000 3000 3000>
    componentLinearAccScale=<0.5 1 1>
    componentLinearVelScale=<0.5 1 1>
    gravityOverride {z=-980}
    physicsBodyDefinitions [{
      sphereCollisionRadius=0
      boundBone {boneName=dyn_hair_f}
      boxExtents=<20 40 40>
      constraintSetup {angularLimitsMax=<0 45 0>, angularLimitsMin=<0 -45 0>, angularTarget=<0 0 0>}
      localJointOffset=<10 0 0>
     }]
    sphericalLimits [
     {limitRadius=10, drivingBone {boneName=spine_03}, sphereLocalOffset=<0 0 0>}
     {limitRadius=10, drivingBone {boneName=spine_04}, sphereLocalOffset=<0 0 0>}
    ]
   }
   n62 = Get IsLongBackblingOnAlpha() -> (IsLongBackblingOnAlpha:double)
   n63 = "Transform (Modify) Bone - Bone: dyn_hair_a"!(ComponentPose=@n61, Alpha:float=@n62, Translation="0,5.5,0", Rotation="0,-12,0", Scale="1,1,1") -> (Pose) {
    rotationMode=BMM_Additive
    rotationSpace=BCS_BoneSpace
    scaleMode=BMM_Additive
    scaleSpace=BCS_BoneSpace
    translationMode=BMM_Additive
    translationSpace=BCS_BoneSpace
    alphaBoolBlend {blendOption=Linear}
    alphaScaleBiasClamp {inRange {max=15}}
    boneToModify {boneName=dyn_hair_a}
    rotation {yaw=-12}
    translation {y=5.5}
   }
   n64 = "Transform (Modify) Bone - Bone: dyn_hair_b"(ComponentPose=@n63, Alpha=@n62, Translation="0,4,0", Rotation="0,5,0", Scale="1,1,1") {
    rotationMode=BMM_Additive
    rotationSpace=BCS_BoneSpace
    scaleMode=BMM_Additive
    scaleSpace=BCS_BoneSpace
    translationMode=BMM_Additive
    translationSpace=BCS_BoneSpace
    alphaBoolBlend {blendOption=Linear}
    alphaScaleBiasClamp {inRange {max=15}}
    boneToModify {boneName=dyn_hair_b}
    rotation {yaw=5}
    translation {y=4}
   }
   n65 = "Transform (Modify) Bone - Bone: dyn_hair_c"(ComponentPose=@n64, Alpha=@n62, Translation="0,1,0", Rotation="0,10,0", Scale="1,1,1") {
    rotationMode=BMM_Additive
    rotationSpace=BCS_BoneSpace
    scaleMode=BMM_Additive
    scaleSpace=BCS_BoneSpace
    translationMode=BMM_Additive
    translationSpace=BCS_BoneSpace
    alphaBoolBlend {blendOption=Linear}
    alphaScaleBiasClamp {inRange {max=15}}
    boneToModify {boneName=dyn_hair_c}
    rotation {yaw=10}
    translation {y=1}
   }
   n66 = "Transform (Modify) Bone - Bone: dyn_hair_d"!(ComponentPose=@n65, Alpha:float=@n62, Translation="0,1,0", Rotation="0,10,0", Scale="1,1,1") -> (Pose) {
    rotationMode=BMM_Additive
    rotationSpace=BCS_BoneSpace
    scaleMode=BMM_Additive
    scaleSpace=BCS_BoneSpace
    translationMode=BMM_Additive
    translationSpace=BCS_BoneSpace
    alphaBoolBlend {blendOption=Linear}
    alphaScaleBiasClamp {inRange {max=15}}
    boneToModify {boneName=dyn_hair_d}
    rotation {yaw=10}
    translation {y=1}
   }
   n67 = "Transform (Modify) Bone - Bone: dyn_hair_e"!(ComponentPose=@n66, Alpha:float=@n62, Translation="0,1,0", Rotation="0,-10,0", Scale="1,1,1") -> (Pose) {
    rotationMode=BMM_Additive
    rotationSpace=BCS_BoneSpace
    scaleMode=BMM_Additive
    scaleSpace=BCS_BoneSpace
    translationMode=BMM_Additive
    translationSpace=BCS_BoneSpace
    alphaBoolBlend {blendOption=Linear}
    alphaScaleBiasClamp {inRange {max=15}}
    boneToModify {boneName=dyn_hair_e}
    rotation {yaw=-10}
    translation {y=1}
   }
   n68 = Get skyDiveFloatPoseFixAlpha()
   n69 = "Transform (Modify) Bone - Bone: dyn_hair_a"(ComponentPose=@n67, Alpha=@n68, Translation="0,0,0", Rotation="0,90,0", Scale="1,1,1") {
    rotationMode=BMM_Additive
    rotationSpace=BCS_BoneSpace
    scaleMode=BMM_Additive
    scaleSpace=BCS_BoneSpace
    translationMode=BMM_Additive
    translationSpace=BCS_BoneSpace
    alphaBoolBlend {blendOption=Linear}
    alphaScaleBiasClamp {outRange {max=0, min=1}}
    boneToModify {boneName=dyn_hair_a}
    rotation {yaw=90}
   }
   n70 = "Transform (Modify) Bone - Bone: dyn_hair_b"(ComponentPose=@n69, Alpha=@n68, Translation="0,0,0", Rotation="0.000,35,0", Scale="1,1,1") {
    rotationMode=BMM_Additive
    rotationSpace=BCS_BoneSpace
    scaleMode=BMM_Additive
    scaleSpace=BCS_BoneSpace
    translationMode=BMM_Additive
    translationSpace=BCS_BoneSpace
    alphaBoolBlend {blendOption=Linear}
    alphaScaleBiasClamp {outRange {max=0, min=1}}
    boneToModify {boneName=dyn_hair_b}
    rotation {yaw=35}
   }
   n71 = "Transform (Modify) Bone - Bone: dyn_hair_c"(
    ComponentPose=@n70
    Alpha=SelectFloat(A=0, B=@n68, bPickA=Get IsBackblingOn() -> (IsBackblingOn:bool))
    Translation="0,0,0"
    Rotation="0,60,0"
    Scale="1,1,1"
   ) {
    rotationMode=BMM_Additive
    rotationSpace=BCS_BoneSpace
    scaleMode=BMM_Additive
    scaleSpace=BCS_BoneSpace
    translationMode=BMM_Additive
    translationSpace=BCS_BoneSpace
    alphaBoolBlend {blendOption=Linear}
    alphaScaleBiasClamp {outRange {max=0, min=1}}
    boneToModify {boneName=dyn_hair_c}
    rotation {yaw=60}
   }
   n72 = Get isJumpPoseFixNeeded() -> (isJumpPoseFixNeeded:double)
   n73 = "Transform (Modify) Bone - Bone: dyn_hair_a"(ComponentPose=@n71, Alpha=@n72, Translation="0,0,0", Rotation="0,15,0", Scale="1,1,1") {
    rotationMode=BMM_Additive
    rotationSpace=BCS_BoneSpace
    translationMode=BMM_Additive
    translationSpace=BCS_BoneSpace
    alphaBoolBlend {blendOption=Linear}
    alphaScaleBiasClamp {bInterpResult=true, interpSpeedDecreasing=5, interpSpeedIncreasing=5, inRange {max=15}}
    boneToModify {boneName=dyn_hair_a}
    rotation {yaw=15}
   }
   n74 = SelectFloat(A=0, B=@n72, bPickA=OR Boolean!(A=Get bIsSkydiveDiveMode(), B=Get bIsSkydiveFloating(), C:bool) -> (ReturnValue))
   n75 = "Transform (Modify) Bone - Bone: dyn_hair_b"(ComponentPose=@n73, Alpha=@n74, Translation="0,0,0", Rotation="0,25,0", Scale="1,1,1") {
    rotationMode=BMM_Additive
    rotationSpace=BCS_BoneSpace
    translationMode=BMM_Additive
    translationSpace=BCS_BoneSpace
    alphaBoolBlend {blendOption=Linear}
    alphaScaleBiasClamp {bInterpResult=true, interpSpeedDecreasing=5, interpSpeedIncreasing=5, inRange {max=15}}
    boneToModify {boneName=dyn_hair_b}
    rotation {yaw=25}
   }
   n76 = "Transform (Modify) Bone - Bone: dyn_hair_c"(ComponentPose=@n75, Alpha=@n74, Translation="0,0,0", Rotation="0,15,0", Scale="1,1,1") {
    rotationMode=BMM_Additive
    rotationSpace=BCS_BoneSpace
    translationMode=BMM_Additive
    translationSpace=BCS_BoneSpace
    alphaBoolBlend {blendOption=Linear}
    alphaScaleBiasClamp {bInterpResult=true, interpSpeedDecreasing=5, interpSpeedIncreasing=5, inRange {max=15}}
    boneToModify {boneName=dyn_hair_c}
    rotation {yaw=15}
   }
   n77 = SelectFloat(A=0, B=@n74, bPickA=Get bIsParachuteOpened())
   n78 = "Transform (Modify) Bone - Bone: dyn_hair_d"(ComponentPose=@n76, Alpha=@n77, Translation="0,0,0", Rotation="0,20,0", Scale="1,1,1") {
    rotationMode=BMM_Additive
    rotationSpace=BCS_BoneSpace
    translationMode=BMM_Additive
    translationSpace=BCS_BoneSpace
    alphaBoolBlend {blendOption=Linear}
    alphaScaleBiasClamp {bInterpResult=true, interpSpeedDecreasing=5, interpSpeedIncreasing=5, inRange {max=15}}
    boneToModify {boneName=dyn_hair_d}
    rotation {yaw=20}
   }
   n79 = "Transform (Modify) Bone - Bone: dyn_hair_e"(ComponentPose=@n78, Alpha=@n77, Translation="0,0,0", Rotation="0,10,0", Scale="1,1,1") {
    rotationMode=BMM_Additive
    rotationSpace=BCS_BoneSpace
    translationMode=BMM_Additive
    translationSpace=BCS_BoneSpace
    alphaBoolBlend {blendOption=Linear}
    alphaScaleBiasClamp {bInterpResult=true, interpSpeedDecreasing=5, interpSpeedIncreasing=5, inRange {max=15}}
    boneToModify {boneName=dyn_hair_e}
    rotation {yaw=10}
   }
   Save cached pose 'SimCachPose'(Pose=Component To Local(ComponentPose=@n79)) {cacheName=SimCachPose}
   n80 = Get hairConstraintON() -> (hairConstraintON:bool)
   n81 = "Constraint - dyn_hair_a (PP)"(
    ComponentPose=Local To Component(LocalPose=Use cached pose 'SharedCachePose'())
    bAlphaBoolEnabled=@n80
    ConstraintWeights_0=0.7
    ConstraintWeights_1=0.3
   ) {
    alphaInputType=Bool
    alphaBoolBlend {blendOption=Linear}
    boneToModify {boneName=dyn_hair_a}
    constraintSetup [{targetBone {boneName=head}}, {targetBone {boneName=spine_05}}]
    constraintWeights [1,1]
   }
   n82 = "Constraint - dyn_hair_b (PP)"(ComponentPose=@n81, bAlphaBoolEnabled=@n80, ConstraintWeights_0=0.3, ConstraintWeights_1=0.7) {
    alphaInputType=Bool
    alphaBoolBlend {blendOption=Linear}
    boneToModify {boneName=dyn_hair_b}
    constraintSetup [{targetBone {boneName=head}}, {targetBone {boneName=spine_05}}]
    constraintWeights [1,1]
   }
   n83 = "Constraint - dyn_hair_c (PPP)"!(ComponentPose=@n82, bAlphaBoolEnabled:bool=@n80, ConstraintWeights_0=0.2, ConstraintWeights_1:float=0.2, ConstraintWeights_2:float=0.6) -> (Pose) {
    alphaInputType=Bool
    alphaBoolBlend {blendOption=Linear}
    boneToModify {boneName=dyn_hair_c}
    constraintSetup [{targetBone {boneName=pelvis}}, {targetBone {boneName=spine_04}}, {targetBone {boneName=spine_05}}]
    constraintWeights [1,1,1]
   }
   n84 = "Constraint - dyn_hair_d (PPP)"!(ComponentPose=@n83, bAlphaBoolEnabled:bool=@n80, ConstraintWeights_0=0.7, ConstraintWeights_1:float=0.2, ConstraintWeights_2:float=0.1) -> (Pose) {
    alphaInputType=Bool
    alphaBoolBlend {blendOption=Linear}
    boneToModify {boneName=dyn_hair_d}
    constraintSetup [{targetBone {boneName=pelvis}}, {targetBone {boneName=spine_04}}, {targetBone {boneName=spine_05}}]
    constraintWeights [1,1,1]
   }
   n85 = "Constraint - dyn_hair_e (PP)"(ComponentPose=@n84, bAlphaBoolEnabled=@n80, ConstraintWeights_0=0.8, ConstraintWeights_1=0.2) {
    alphaInputType=Bool
    alphaBoolBlend {blendOption=Linear}
    boneToModify {boneName=dyn_hair_e}
    constraintSetup [{targetBone {boneName=pelvis}}, {targetBone {boneName=spine_02}}]
    constraintWeights [1,1]
   }
   n86 = Get hairConstraintON()
   n87 = "Transform (Modify) Bone - Bone: dyn_hair_b"!(ComponentPose=@n85, bAlphaBoolEnabled:bool=@n86, Translation="0,0,-5", Rotation="5,0,0", Scale="1,1,1") -> (Pose) {
    alphaInputType=Bool
    rotationMode=BMM_Additive
    rotationSpace=BCS_BoneSpace
    translationMode=BMM_Additive
    translationSpace=BCS_BoneSpace
    alphaBoolBlend {blendOption=Linear}
    alphaScaleBiasClamp {bMapRange=true, inRange {max=0, min=-25}, outRange {max=0, min=1}}
    boneToModify {boneName=dyn_hair_b}
    rotation {pitch=5}
    translation {z=-5}
   }
   n88 = "Constraint - dyn_hair_small_a (PPP)"!(ComponentPose=@n87, bAlphaBoolEnabled:bool=@n86, ConstraintWeights_0=0.2, ConstraintWeights_1:float=0.7, ConstraintWeights_2:float=0.1) -> (Pose) {
    alphaInputType=Bool
    alphaBoolBlend {blendOption=Linear}
    boneToModify {boneName=dyn_hair_small_a}
    constraintSetup [{targetBone {boneName=dyn_hair_c}}, {targetBone {boneName=dyn_hair_b}}, {targetBone {boneName=dyn_hair_a}}]
    constraintWeights [1,1,1]
   }
   n89 = "Constraint - dyn_hair_small_b (PP)"!(ComponentPose=@n88, bAlphaBoolEnabled:bool=@n86, ConstraintWeights_0=0.3, ConstraintWeights_1:float=0.7) -> (Pose) {
    alphaInputType=Bool
    alphaBoolBlend {blendOption=Linear}
    boneToModify {boneName=dyn_hair_small_b}
    constraintSetup [{targetBone {boneName=dyn_hair_d}}, {targetBone {boneName=dyn_hair_c}}]
    constraintWeights [1,1]
   }
   n90 = "Constraint - dyn_hair_small_c (PPP)"!(ComponentPose=@n89, bAlphaBoolEnabled:bool=@n86, ConstraintWeights_0=0.1, ConstraintWeights_1:float=0.8, ConstraintWeights_2:float=0.1) -> (Pose) {
    alphaInputType=Bool
    alphaBoolBlend {blendOption=Linear}
    boneToModify {boneName=dyn_hair_small_c}
    constraintSetup [{targetBone {boneName=dyn_hair_d}}, {targetBone {boneName=dyn_hair_c}}, {targetBone {boneName=dyn_hair_b}}]
    constraintWeights [1,1,1]
   }
   n91 = Get IsShortBackblingOn()
   n92 = "Transform (Modify) Bone - Bone: dyn_hair_band_base"!(ComponentPose=@n90, bAlphaBoolEnabled:bool=@n91, Translation="0,1,0", Rotation="12,0,10", Scale="0.8,1,1") -> (Pose) {
    alphaInputType=Bool
    rotationMode=BMM_Additive
    rotationSpace=BCS_BoneSpace
    scaleMode=BMM_Additive
    scaleSpace=BCS_BoneSpace
    translationMode=BMM_Additive
    translationSpace=BCS_BoneSpace
    alphaBoolBlend {blendOption=Linear}
    alphaScaleBiasClamp {bMapRange=true, inRange {max=0, min=-25}, outRange {max=0, min=1}}
    boneToModify {boneName=dyn_hair_band_base}
    rotation {pitch=12, roll=10}
    scale {x=0.8}
    translation {y=1}
   }
   n93 = "Transform (Modify) Bone - Bone: dyn_hair_a"!(ComponentPose=@n92, bAlphaBoolEnabled:bool=@n91, Translation="0,2,0", Rotation="0,-10,0", Scale="1,1,1") -> (Pose) {
    alphaInputType=Bool
    rotationMode=BMM_Additive
    rotationSpace=BCS_BoneSpace
    scaleMode=BMM_Additive
    scaleSpace=BCS_BoneSpace
    translationMode=BMM_Additive
    translationSpace=BCS_BoneSpace
    alphaBoolBlend {blendOption=Linear}
    alphaScaleBiasClamp {bMapRange=true, inRange {max=0, min=-25}, outRange {max=0, min=1}}
    boneToModify {boneName=dyn_hair_a}
    rotation {yaw=-10}
    translation {y=2}
   }
   n94 = "Transform (Modify) Bone - Bone: dyn_hair_b"(ComponentPose=@n93, bAlphaBoolEnabled=@n91, Translation="-2,0,0", Rotation="-5,0,0", Scale="1,1,1") {
    alphaInputType=Bool
    rotationMode=BMM_Additive
    rotationSpace=BCS_BoneSpace
    scaleMode=BMM_Additive
    scaleSpace=BCS_BoneSpace
    translationMode=BMM_Additive
    translationSpace=BCS_BoneSpace
    alphaBoolBlend {blendOption=Linear}
    alphaScaleBiasClamp {bMapRange=true, inRange {max=0, min=-25}, outRange {max=0, min=1}}
    boneToModify {boneName=dyn_hair_b}
    rotation {pitch=-5}
    translation {x=-2}
   }
   n95 = "Transform (Modify) Bone - Bone: dyn_hair_c"!(ComponentPose=@n94, bAlphaBoolEnabled:bool=@n91, Translation="-1.5,0,0", Rotation="-20,0,0", Scale="0.5,0.8,0.8") -> (Pose) {
    alphaInputType=Bool
    rotationMode=BMM_Additive
    rotationSpace=BCS_BoneSpace
    scaleMode=BMM_Additive
    scaleSpace=BCS_BoneSpace
    translationMode=BMM_Additive
    translationSpace=BCS_BoneSpace
    alphaBoolBlend {blendOption=Linear}
    alphaScaleBiasClamp {bMapRange=true, inRange {max=0, min=-25}, outRange {max=0, min=1}}
    boneToModify {boneName=dyn_hair_c}
    rotation {pitch=-20}
    scale=<0.5 0.8 0.8>
    translation {x=-1.5}
   }
   n96 = "Transform (Modify) Bone - Bone: dyn_hair_d"!(ComponentPose=@n95, bAlphaBoolEnabled:bool=@n91, Translation="-1.5,-1,0", Rotation="0,-10,0", Scale="0.5,0.8,0.8") -> (Pose) {
    alphaInputType=Bool
    rotationMode=BMM_Additive
    rotationSpace=BCS_BoneSpace
    scaleMode=BMM_Additive
    scaleSpace=BCS_BoneSpace
    translationMode=BMM_Additive
    translationSpace=BCS_BoneSpace
    alphaBoolBlend {blendOption=Linear}
    alphaScaleBiasClamp {bMapRange=true, inRange {max=0, min=-25}, outRange {max=0, min=1}}
    boneToModify {boneName=dyn_hair_d}
    rotation {yaw=-10}
    scale=<0.5 0.8 0.8>
    translation=<-1.5 -1>
   }
   n97 = "Transform (Modify) Bone - Bone: dyn_hair_e"!(ComponentPose=@n96, bAlphaBoolEnabled:bool=@n91, Translation="0,0,0", Rotation="0,20,0", Scale="1,1,1") -> (Pose) {
    alphaInputType=Bool
    rotationMode=BMM_Additive
    rotationSpace=BCS_BoneSpace
    scaleMode=BMM_Additive
    scaleSpace=BCS_BoneSpace
    translationMode=BMM_Additive
    translationSpace=BCS_BoneSpace
    alphaBoolBlend {blendOption=Linear}
    alphaScaleBiasClamp {bMapRange=true, inRange {max=0, min=-25}, outRange {max=0, min=1}}
    boneToModify {boneName=dyn_hair_e}
    rotation {yaw=20}
   }
   n98 = Get IsShortBackblingOn()
   n99 = "Transform (Modify) Bone - Bone: dyn_hair_small_a"!(ComponentPose=@n97, bAlphaBoolEnabled:bool=@n98, Translation="0,2,0", Rotation="0,0,0", Scale="0.85,0.85,0.85") -> (Pose) {
    alphaInputType=Bool
    rotationMode=BMM_Additive
    rotationSpace=BCS_BoneSpace
    scaleMode=BMM_Additive
    scaleSpace=BCS_BoneSpace
    translationMode=BMM_Additive
    translationSpace=BCS_BoneSpace
    alphaBoolBlend {blendOption=Linear}
    alphaScaleBiasClamp {bMapRange=true, inRange {max=0, min=-25}, outRange {max=0, min=1}}
    boneToModify {boneName=dyn_hair_small_a}
    scale=<0.85 0.85 0.85>
    translation {y=2}
   }
   n100 = "Transform (Modify) Bone - Bone: dyn_hair_small_b"(ComponentPose=@n99, bAlphaBoolEnabled=@n98, Translation="0,-4,0", Rotation="0,5,0", Scale="0.8,1,1") {
    alphaInputType=Bool
    rotationMode=BMM_Additive
    rotationSpace=BCS_BoneSpace
    scaleMode=BMM_Additive
    scaleSpace=BCS_BoneSpace
    translationMode=BMM_Additive
    translationSpace=BCS_BoneSpace
    alphaBoolBlend {blendOption=Linear}
    alphaScaleBiasClamp {bMapRange=true, inRange {max=0, min=-25}, outRange {max=0, min=1}}
    boneToModify {boneName=dyn_hair_small_b}
    rotation {yaw=5}
    scale {x=0.8}
    translation {y=-4}
   }
   n101 = "Transform (Modify) Bone - Bone: dyn_hair_small_c"(ComponentPose=@n100, bAlphaBoolEnabled=@n98, Translation="0,-3,0", Rotation="0,0,0", Scale="0.75,1,1") {
    alphaInputType=Bool
    rotationMode=BMM_Additive
    rotationSpace=BCS_BoneSpace
    scaleMode=BMM_Additive
    scaleSpace=BCS_BoneSpace
    translationMode=BMM_Additive
    translationSpace=BCS_BoneSpace
    alphaBoolBlend {blendOption=Linear}
    alphaScaleBiasClamp {bMapRange=true, inRange {max=0, min=-25}, outRange {max=0, min=1}}
    boneToModify {boneName=dyn_hair_small_c}
    scale {x=0.75}
    translation {y=-3}
   }
   n102 = "Transform (Modify) Bone - Bone: dyn_hair_e"(ComponentPose=@n101, bAlphaBoolEnabled=@n98, Translation="-1.5,0,0", Rotation="0,0,0", Scale="0.4,0.8,0.8") {
    alphaInputType=Bool
    rotationMode=BMM_Additive
    rotationSpace=BCS_BoneSpace
    scaleMode=BMM_Additive
    scaleSpace=BCS_BoneSpace
    translationMode=BMM_Additive
    translationSpace=BCS_BoneSpace
    alphaBoolBlend {blendOption=Linear}
    alphaScaleBiasClamp {bMapRange=true, inRange {max=0, min=-25}, outRange {max=0, min=1}}
    boneToModify {boneName=dyn_hair_e}
    scale=<0.4 0.8 0.8>
    translation {x=-1.5}
   }
   n103 = "Transform (Modify) Bone - Bone: dyn_hair_band_base"!(ComponentPose=@n102, Alpha:float=@n18, Translation="0,0,0", Rotation="0,-30,0", Scale="1.2,1.05,1.05") -> (Pose) {
    rotationMode=BMM_Additive
    rotationSpace=BCS_BoneSpace
    scaleMode=BMM_Additive
    scaleSpace=BCS_BoneSpace
    translationMode=BMM_Additive
    translationSpace=BCS_BoneSpace
    alphaBoolBlend {blendOption=Linear}
    alphaScaleBiasClamp {bMapRange=true, inRange {max=0, min=-25}, outRange {max=0, min=1}}
    boneToModify {boneName=dyn_hair_band_base}
    rotation {yaw=-30}
    scale=<1.2 1.05 1.05>
   }
   n104 = "Transform (Modify) Bone - Bone: dyn_hair_a"!(ComponentPose=@n103, Alpha:float=@n18, Translation="0,1,0", Rotation="10,-30,0", Scale="1,1,1") -> (Pose) {
    rotationMode=BMM_Additive
    rotationSpace=BCS_BoneSpace
    scaleMode=BMM_Additive
    scaleSpace=BCS_BoneSpace
    translationMode=BMM_Additive
    translationSpace=BCS_BoneSpace
    alphaBoolBlend {blendOption=Linear}
    alphaScaleBiasClamp {bMapRange=true, inRange {max=0, min=-25}, outRange {max=0, min=1}}
    boneToModify {boneName=dyn_hair_a}
    rotation {pitch=10, yaw=-30}
    translation {y=1}
   }
   n105 = "Transform (Modify) Bone - Bone: dyn_hair_c"!(ComponentPose=@n104, Alpha:float=@n18, Translation="0,0,0", Rotation="-10,0,0", Scale="1,1,1") -> (Pose) {
    rotationMode=BMM_Additive
    rotationSpace=BCS_BoneSpace
    scaleMode=BMM_Additive
    scaleSpace=BCS_BoneSpace
    translationMode=BMM_Additive
    translationSpace=BCS_BoneSpace
    alphaBoolBlend {blendOption=Linear}
    alphaScaleBiasClamp {bMapRange=true, inRange {max=0, min=-25}, outRange {max=0, min=1}}
    boneToModify {boneName=dyn_hair_c}
    rotation {pitch=-10}
   }
   n106 = Get backpackOnHeadUpDn()
   n107 = "Transform (Modify) Bone - Bone: dyn_hair_a"(ComponentPose=@n105, Alpha=@n106, Translation="0,1,0", Rotation="0,0,0", Scale="1,1,1") {
    rotationMode=BMM_Additive
    rotationSpace=BCS_BoneSpace
    scaleMode=BMM_Additive
    scaleSpace=BCS_BoneSpace
    translationMode=BMM_Additive
    translationSpace=BCS_BoneSpace
    alphaBoolBlend {blendOption=Linear}
    alphaScaleBiasClamp {bMapRange=true, inRange {max=15}}
    boneToModify {boneName=dyn_hair_a}
    translation {y=1}
   }
   n108 = "Transform (Modify) Bone - Bone: dyn_hair_b"!(ComponentPose=@n107, Alpha:float=@n106, Translation="0,0,0", Rotation="0,2,0", Scale="1,1,1") -> (Pose) {
    rotationMode=BMM_Additive
    rotationSpace=BCS_BoneSpace
    scaleMode=BMM_Additive
    scaleSpace=BCS_BoneSpace
    translationMode=BMM_Additive
    translationSpace=BCS_BoneSpace
    alphaBoolBlend {blendOption=Linear}
    alphaScaleBiasClamp {bMapRange=true, inRange {max=15}}
    boneToModify {boneName=dyn_hair_b}
    rotation {yaw=2}
   }
   n109 = "Transform (Modify) Bone - Bone: dyn_hair_small_b"!(ComponentPose=@n108, Alpha:float=@n106, Translation="0,0,0", Rotation="0,5,0", Scale="1,1,1") -> (Pose) {
    rotationMode=BMM_Additive
    rotationSpace=BCS_BoneSpace
    scaleMode=BMM_Additive
    scaleSpace=BCS_BoneSpace
    translationMode=BMM_Additive
    translationSpace=BCS_BoneSpace
    alphaBoolBlend {blendOption=Linear}
    alphaScaleBiasClamp {bMapRange=true, inRange {max=15}}
    boneToModify {boneName=dyn_hair_small_b}
    rotation {yaw=5}
   }
   n110 = "Transform (Modify) Bone - Bone: dyn_hair_d"(ComponentPose=@n109, bAlphaBoolEnabled=Get bIsParachuteOpened(), Translation="0,0,0", Rotation="0,-30,0", Scale="1,1,1") {
    alphaInputType=Bool
    rotationMode=BMM_Additive
    rotationSpace=BCS_BoneSpace
    scaleMode=BMM_Additive
    scaleSpace=BCS_BoneSpace
    translationMode=BMM_Additive
    translationSpace=BCS_BoneSpace
    alphaBoolBlend {blendOption=Linear}
    alphaScaleBiasClamp {bMapRange=true, inRange {max=5, min=-15}}
    boneToModify {boneName=dyn_hair_d}
    rotation {yaw=-30}
   }
   n111 = Get isBistroGlider() -> (isBistroGlider:double)
   n112 = "Transform (Modify) Bone - Bone: dyn_hair_a"(ComponentPose=@n110, Alpha=@n111, Translation="0,0,0", Rotation="0,35,0", Scale="1,1,1") {
    rotationMode=BMM_Additive
    rotationSpace=BCS_BoneSpace
    translationMode=BMM_Additive
    translationSpace=BCS_BoneSpace
    alphaBoolBlend {blendOption=Linear}
    alphaScaleBiasClamp {bMapRange=true}
    boneToModify {boneName=dyn_hair_a}
    rotation {yaw=35}
   }
   n113 = "Transform (Modify) Bone - Bone: dyn_hair_b"(ComponentPose=@n112, Alpha=@n111, Translation="0,0,0", Rotation="0,20,0", Scale="1,1,1") {
    rotationMode=BMM_Additive
    rotationSpace=BCS_BoneSpace
    translationMode=BMM_Additive
    translationSpace=BCS_BoneSpace
    alphaBoolBlend {blendOption=Linear}
    alphaScaleBiasClamp {bMapRange=true}
    boneToModify {boneName=dyn_hair_b}
    rotation {yaw=20}
   }
   n114 = "Transform (Modify) Bone - Bone: dyn_hair_a_base"(ComponentPose=@n113, Alpha=@n111, Translation="0,0,0", Rotation="0,10,0", Scale="1,1,1") {
    rotationMode=BMM_Additive
    rotationSpace=BCS_BoneSpace
    translationMode=BMM_Additive
    translationSpace=BCS_BoneSpace
    alphaBoolBlend {blendOption=Linear}
    alphaScaleBiasClamp {bMapRange=true}
    boneToModify {boneName=dyn_hair_a_base}
    rotation {yaw=10}
   }
   n115 = "Transform (Modify) Bone - Bone: dyn_hair_c"(ComponentPose=@n114, Alpha=Get head_fwdBwd_rot(), Translation="0,0,0", Rotation="0,5,0", Scale="1,1,1") {
    rotationMode=BMM_Additive
    rotationSpace=BCS_BoneSpace
    scaleMode=BMM_Additive
    scaleSpace=BCS_BoneSpace
    translationMode=BMM_Additive
    translationSpace=BCS_BoneSpace
    alphaBoolBlend {blendOption=Linear}
    alphaScaleBiasClamp {bMapRange=true, inRange {max=5, min=-15}}
    boneToModify {boneName=dyn_hair_c}
    rotation {yaw=5}
   }
   Save cached pose 'Backbling Cache'(Pose=Component To Local(ComponentPose=@n115)) {cacheName=Backbling Cache}
   n116 = "Constraint - dyn_hair_a (PP)"(
    ComponentPose=Local To Component(LocalPose=Use cached pose 'SharedCachePose'())
    bAlphaBoolEnabled=True
    ConstraintWeights_0=0.7
    ConstraintWeights_1=0.3
   ) {
    alphaInputType=Bool
    alphaBoolBlend {blendOption=Linear}
    boneToModify {boneName=dyn_hair_a}
    constraintSetup [{targetBone {boneName=head}}, {targetBone {boneName=spine_05}}]
    constraintWeights [1,1]
   }
   n117 = "Constraint - dyn_hair_b (PP)"(ComponentPose=@n116, bAlphaBoolEnabled=True, ConstraintWeights_0=0.3, ConstraintWeights_1=0.7) {
    alphaInputType=Bool
    alphaBoolBlend {blendOption=Linear}
    boneToModify {boneName=dyn_hair_b}
    constraintSetup [{targetBone {boneName=head}}, {targetBone {boneName=spine_05}}]
    constraintWeights [1,1]
   }
   n118 = "Constraint - dyn_hair_c (PPP)"(ComponentPose=@n117, bAlphaBoolEnabled=True, ConstraintWeights_0=0.2, ConstraintWeights_1=0.2, ConstraintWeights_2=0.6) {
    alphaInputType=Bool
    alphaBoolBlend {blendOption=Linear}
    boneToModify {boneName=dyn_hair_c}
    constraintSetup [{targetBone {boneName=pelvis}}, {targetBone {boneName=spine_04}}, {targetBone {boneName=spine_05}}]
    constraintWeights [1,1,1]
   }
   n119 = "Constraint - dyn_hair_d (PPP)"(ComponentPose=@n118, bAlphaBoolEnabled=True, ConstraintWeights_0=0.7, ConstraintWeights_1=0.2, ConstraintWeights_2=0.1) {
    alphaInputType=Bool
    alphaBoolBlend {blendOption=Linear}
    boneToModify {boneName=dyn_hair_d}
    constraintSetup [{targetBone {boneName=pelvis}}, {targetBone {boneName=spine_04}}, {targetBone {boneName=spine_05}}]
    constraintWeights [1,1,1]
   }
   n120 = "Constraint - dyn_hair_e (PP)"(ComponentPose=@n119, bAlphaBoolEnabled=True, ConstraintWeights_0=0.8, ConstraintWeights_1=0.2) {
    alphaInputType=Bool
    alphaBoolBlend {blendOption=Linear}
    boneToModify {boneName=dyn_hair_e}
    constraintSetup [{targetBone {boneName=pelvis}}, {targetBone {boneName=spine_02}}]
    constraintWeights [1,1]
   }
   n121 = "Transform (Modify) Bone - Bone: dyn_hair_b"!(ComponentPose=@n120, bAlphaBoolEnabled:bool=True, Translation="0,0,-5", Rotation="5,0,0", Scale="1,1,1") -> (Pose) {
    alphaInputType=Bool
    rotationMode=BMM_Additive
    rotationSpace=BCS_BoneSpace
    translationMode=BMM_Additive
    translationSpace=BCS_BoneSpace
    alphaBoolBlend {blendOption=Linear}
    alphaScaleBiasClamp {bMapRange=true, inRange {max=0, min=-25}, outRange {max=0, min=1}}
    boneToModify {boneName=dyn_hair_b}
    rotation {pitch=5}
    translation {z=-5}
   }
   n122 = "Constraint - dyn_hair_small_a (PPP)"(ComponentPose=@n121, bAlphaBoolEnabled=True, ConstraintWeights_0=0.2, ConstraintWeights_1=0.7, ConstraintWeights_2=0.1) {
    alphaInputType=Bool
    alphaBoolBlend {blendOption=Linear}
    boneToModify {boneName=dyn_hair_small_a}
    constraintSetup [{targetBone {boneName=dyn_hair_c}}, {targetBone {boneName=dyn_hair_b}}, {targetBone {boneName=dyn_hair_a}}]
    constraintWeights [1,1,1]
   }
   n123 = "Constraint - dyn_hair_small_b (PP)"(ComponentPose=@n122, bAlphaBoolEnabled=True, ConstraintWeights_0=0.3, ConstraintWeights_1=0.7) {
    alphaInputType=Bool
    alphaBoolBlend {blendOption=Linear}
    boneToModify {boneName=dyn_hair_small_b}
    constraintSetup [{targetBone {boneName=dyn_hair_d}}, {targetBone {boneName=dyn_hair_c}}]
    constraintWeights [1,1]
   }
   n124 = "Constraint - dyn_hair_small_c (PPP)"(ComponentPose=@n123, bAlphaBoolEnabled=True, ConstraintWeights_0=0.1, ConstraintWeights_1=0.8, ConstraintWeights_2=0.1) {
    alphaInputType=Bool
    alphaBoolBlend {blendOption=Linear}
    boneToModify {boneName=dyn_hair_small_c}
    constraintSetup [{targetBone {boneName=dyn_hair_d}}, {targetBone {boneName=dyn_hair_c}}, {targetBone {boneName=dyn_hair_b}}]
    constraintWeights [1,1,1]
   }
   n125 = Get IsShortBackblingOn()
   n126 = "Transform (Modify) Bone - Bone: dyn_hair_band_base"!(ComponentPose=@n124, bAlphaBoolEnabled:bool=@n125, Translation="0,1,0", Rotation="12,0,10", Scale="0.8,1,1") -> (Pose) {
    alphaInputType=Bool
    rotationMode=BMM_Additive
    rotationSpace=BCS_BoneSpace
    scaleMode=BMM_Additive
    scaleSpace=BCS_BoneSpace
    translationMode=BMM_Additive
    translationSpace=BCS_BoneSpace
    alphaBoolBlend {blendOption=Linear}
    alphaScaleBiasClamp {bMapRange=true, inRange {max=0, min=-25}, outRange {max=0, min=1}}
    boneToModify {boneName=dyn_hair_band_base}
    rotation {pitch=12, roll=10}
    scale {x=0.8}
    translation {y=1}
   }
   n127 = "Transform (Modify) Bone - Bone: dyn_hair_a"!(ComponentPose=@n126, bAlphaBoolEnabled:bool=@n125, Translation="0,2,0", Rotation="0,-10,0", Scale="1,1,1") -> (Pose) {
    alphaInputType=Bool
    rotationMode=BMM_Additive
    rotationSpace=BCS_BoneSpace
    scaleMode=BMM_Additive
    scaleSpace=BCS_BoneSpace
    translationMode=BMM_Additive
    translationSpace=BCS_BoneSpace
    alphaBoolBlend {blendOption=Linear}
    alphaScaleBiasClamp {bMapRange=true, inRange {max=0, min=-25}, outRange {max=0, min=1}}
    boneToModify {boneName=dyn_hair_a}
    rotation {yaw=-10}
    translation {y=2}
   }
   n128 = "Transform (Modify) Bone - Bone: dyn_hair_b"(ComponentPose=@n127, bAlphaBoolEnabled=@n125, Translation="-2,0,0", Rotation="-5,0,0", Scale="1,1,1") {
    alphaInputType=Bool
    rotationMode=BMM_Additive
    rotationSpace=BCS_BoneSpace
    scaleMode=BMM_Additive
    scaleSpace=BCS_BoneSpace
    translationMode=BMM_Additive
    translationSpace=BCS_BoneSpace
    alphaBoolBlend {blendOption=Linear}
    alphaScaleBiasClamp {bMapRange=true, inRange {max=0, min=-25}, outRange {max=0, min=1}}
    boneToModify {boneName=dyn_hair_b}
    rotation {pitch=-5}
    translation {x=-2}
   }
   n129 = "Transform (Modify) Bone - Bone: dyn_hair_c"!(ComponentPose=@n128, bAlphaBoolEnabled:bool=@n125, Translation="-1.5,0,0", Rotation="-20,0,0", Scale="0.5,0.8,0.8") -> (Pose) {
    alphaInputType=Bool
    rotationMode=BMM_Additive
    rotationSpace=BCS_BoneSpace
    scaleMode=BMM_Additive
    scaleSpace=BCS_BoneSpace
    translationMode=BMM_Additive
    translationSpace=BCS_BoneSpace
    alphaBoolBlend {blendOption=Linear}
    alphaScaleBiasClamp {bMapRange=true, inRange {max=0, min=-25}, outRange {max=0, min=1}}
    boneToModify {boneName=dyn_hair_c}
    rotation {pitch=-20}
    scale=<0.5 0.8 0.8>
    translation {x=-1.5}
   }
   n130 = "Transform (Modify) Bone - Bone: dyn_hair_d"(ComponentPose=@n129, bAlphaBoolEnabled=@n125, Translation="-1.5,-1,0", Rotation="0,-10,0", Scale="0.5,0.8,0.8") {
    alphaInputType=Bool
    rotationMode=BMM_Additive
    rotationSpace=BCS_BoneSpace
    scaleMode=BMM_Additive
    scaleSpace=BCS_BoneSpace
    translationMode=BMM_Additive
    translationSpace=BCS_BoneSpace
    alphaBoolBlend {blendOption=Linear}
    alphaScaleBiasClamp {bMapRange=true, inRange {max=0, min=-25}, outRange {max=0, min=1}}
    boneToModify {boneName=dyn_hair_d}
    rotation {yaw=-10}
    scale=<0.5 0.8 0.8>
    translation=<-1.5 -1>
   }
   n131 = "Transform (Modify) Bone - Bone: dyn_hair_e"(ComponentPose=@n130, bAlphaBoolEnabled=@n125, Translation="0,0,0", Rotation="0,20,0", Scale="1,1,1") {
    alphaInputType=Bool
    rotationMode=BMM_Additive
    rotationSpace=BCS_BoneSpace
    scaleMode=BMM_Additive
    scaleSpace=BCS_BoneSpace
    translationMode=BMM_Additive
    translationSpace=BCS_BoneSpace
    alphaBoolBlend {blendOption=Linear}
    alphaScaleBiasClamp {bMapRange=true, inRange {max=0, min=-25}, outRange {max=0, min=1}}
    boneToModify {boneName=dyn_hair_e}
    rotation {yaw=20}
   }
   n132 = Get IsShortBackblingOn()
   n133 = "Transform (Modify) Bone - Bone: dyn_hair_small_a"(ComponentPose=@n131, bAlphaBoolEnabled=@n132, Translation="0,2,0", Rotation="0,0,0", Scale="0.85,0.85,0.85") {
    alphaInputType=Bool
    rotationMode=BMM_Additive
    rotationSpace=BCS_BoneSpace
    scaleMode=BMM_Additive
    scaleSpace=BCS_BoneSpace
    translationMode=BMM_Additive
    translationSpace=BCS_BoneSpace
    alphaBoolBlend {blendOption=Linear}
    alphaScaleBiasClamp {bMapRange=true, inRange {max=0, min=-25}, outRange {max=0, min=1}}
    boneToModify {boneName=dyn_hair_small_a}
    scale=<0.85 0.85 0.85>
    translation {y=2}
   }
   n134 = "Transform (Modify) Bone - Bone: dyn_hair_small_b"!(ComponentPose=@n133, bAlphaBoolEnabled:bool=@n132, Translation="0,-4,0", Rotation="0,5,0", Scale="0.8,1,1") -> (Pose) {
    alphaInputType=Bool
    rotationMode=BMM_Additive
    rotationSpace=BCS_BoneSpace
    scaleMode=BMM_Additive
    scaleSpace=BCS_BoneSpace
    translationMode=BMM_Additive
    translationSpace=BCS_BoneSpace
    alphaBoolBlend {blendOption=Linear}
    alphaScaleBiasClamp {bMapRange=true, inRange {max=0, min=-25}, outRange {max=0, min=1}}
    boneToModify {boneName=dyn_hair_small_b}
    rotation {yaw=5}
    scale {x=0.8}
    translation {y=-4}
   }
   n135 = "Transform (Modify) Bone - Bone: dyn_hair_small_c"(ComponentPose=@n134, bAlphaBoolEnabled=@n132, Translation="0,-3,0", Rotation="0,0,0", Scale="0.75,1,1") {
    alphaInputType=Bool
    rotationMode=BMM_Additive
    rotationSpace=BCS_BoneSpace
    scaleMode=BMM_Additive
    scaleSpace=BCS_BoneSpace
    translationMode=BMM_Additive
    translationSpace=BCS_BoneSpace
    alphaBoolBlend {blendOption=Linear}
    alphaScaleBiasClamp {bMapRange=true, inRange {max=0, min=-25}, outRange {max=0, min=1}}
    boneToModify {boneName=dyn_hair_small_c}
    scale {x=0.75}
    translation {y=-3}
   }
   n136 = "Transform (Modify) Bone - Bone: dyn_hair_e"(ComponentPose=@n135, bAlphaBoolEnabled=@n132, Translation="-1.5,0,0", Rotation="0,0,0", Scale="0.4,0.8,0.8") {
    alphaInputType=Bool
    rotationMode=BMM_Additive
    rotationSpace=BCS_BoneSpace
    scaleMode=BMM_Additive
    scaleSpace=BCS_BoneSpace
    translationMode=BMM_Additive
    translationSpace=BCS_BoneSpace
    alphaBoolBlend {blendOption=Linear}
    alphaScaleBiasClamp {bMapRange=true, inRange {max=0, min=-25}, outRange {max=0, min=1}}
    boneToModify {boneName=dyn_hair_e}
    scale=<0.4 0.8 0.8>
    translation {x=-1.5}
   }
   n137 = Get isBistroGlider()
   n138 = "Transform (Modify) Bone - Bone: dyn_hair_a"!(ComponentPose=@n136, Alpha:float=@n137, Translation="0,0,0", Rotation="0,35,0", Scale="1,1,1") -> (Pose) {
    rotationMode=BMM_Additive
    rotationSpace=BCS_BoneSpace
    translationMode=BMM_Additive
    translationSpace=BCS_BoneSpace
    alphaBoolBlend {blendOption=Linear}
    alphaScaleBiasClamp {bMapRange=true}
    boneToModify {boneName=dyn_hair_a}
    rotation {yaw=35}
   }
   n139 = "Transform (Modify) Bone - Bone: dyn_hair_b"!(ComponentPose=@n138, Alpha:float=@n137, Translation="0,0,0", Rotation="0,20,0", Scale="1,1,1") -> (Pose) {
    rotationMode=BMM_Additive
    rotationSpace=BCS_BoneSpace
    translationMode=BMM_Additive
    translationSpace=BCS_BoneSpace
    alphaBoolBlend {blendOption=Linear}
    alphaScaleBiasClamp {bMapRange=true}
    boneToModify {boneName=dyn_hair_b}
    rotation {yaw=20}
   }
   n140 = "Transform (Modify) Bone - Bone: dyn_hair_a_base"(ComponentPose=@n139, Alpha=@n137, Translation="0,0,0", Rotation="0,10,0", Scale="1,1,1") {
    rotationMode=BMM_Additive
    rotationSpace=BCS_BoneSpace
    translationMode=BMM_Additive
    translationSpace=BCS_BoneSpace
    alphaBoolBlend {blendOption=Linear}
    alphaScaleBiasClamp {bMapRange=true}
    boneToModify {boneName=dyn_hair_a_base}
    rotation {yaw=10}
   }
   n141 = "Transform (Modify) Bone - Bone: dyn_hair_c"!(ComponentPose=@n140, Alpha:float=Get head_fwdBwd_rot(), Translation="0,0,0", Rotation="0,7,0", Scale="1,1,1") -> (Pose) {
    rotationMode=BMM_Additive
    rotationSpace=BCS_BoneSpace
    scaleMode=BMM_Additive
    scaleSpace=BCS_BoneSpace
    translationMode=BMM_Additive
    translationSpace=BCS_BoneSpace
    alphaBoolBlend {blendOption=Linear}
    alphaScaleBiasClamp {bMapRange=true, inRange {max=5, min=-15}}
    boneToModify {boneName=dyn_hair_c}
    rotation {yaw=7}
   }
   Save cached pose 'NO Backbling Cache'(Pose=Component To Local(ComponentPose=@n141)) {cacheName=NO Backbling Cache}
   n142 = Layered blend per bone(
    BasePose=Copy Pose From Mesh() {bCopyCurves=true, bUseAttachedParent=true}
    BlendPoses_0=Copy Pose From Mesh(SourceMeshComponent=Get Tail_Mesh() -> (Tail_Mesh:Skeletal Mesh Component Object Reference)) {bCopyCurves=true, bUseAttachedParent=true}
   ) {layerSetup [{branchFilters [{boneName=None}, {boneName=tail_base}, {boneName=tail_1}, {boneName=tail_2}, {boneName=tail_3}]}]}
   n143 = Layered blend per bone(BasePose=@n142, BlendPoses_0=Copy Pose From Mesh(SourceMeshComponent=Get Body_Mesh()) {bCopyCurves=true, bUseAttachedParent=true}) {layerSetup [{branchFilters [
       {boneName=upperarm_l}
       {boneName=upperarm_r}
       {boneName=lowerarm_l}
       {boneName=lowerarm_r}
       {boneName=thigh_l}
       {boneName=thigh_r}
       {boneName=calf_l}
       {boneName=calf_r}
      ]}]}
   n144 = Get skyDiveFloatPoseFixAlpha()
   n145 = "Transform (Modify) Bone - Bone: dyn_hair_a_base"(ComponentPose=Local To Component(LocalPose=@n143), Alpha=@n144, Translation="0,0,0", Rotation="0,-45,0", Scale="1,1,1") {
    rotationMode=BMM_Additive
    rotationSpace=BCS_BoneSpace
    scaleMode=BMM_Additive
    scaleSpace=BCS_BoneSpace
    translationMode=BMM_Additive
    translationSpace=BCS_BoneSpace
    alphaBoolBlend {blendOption=Linear}
    alphaScaleBiasClamp {inRange {max=0, min=-30}, outRange {max=0, min=1}}
    boneToModify {boneName=dyn_hair_a_base}
    rotation {yaw=-45}
   }
   n146 = "Transform (Modify) Bone - Bone: dyn_hair_a_base"(ComponentPose=@n145, Alpha=@n144, Translation="0,0,0", Rotation="0,20,0", Scale="1,1,1") {
    rotationMode=BMM_Additive
    rotationSpace=BCS_BoneSpace
    scaleMode=BMM_Additive
    scaleSpace=BCS_BoneSpace
    translationMode=BMM_Additive
    translationSpace=BCS_BoneSpace
    alphaBoolBlend {blendOption=Linear}
    alphaScaleBiasClamp {inRange {max=30}}
    boneToModify {boneName=dyn_hair_a_base}
    rotation {yaw=20}
   }
   n147 = Get hairConstraintON()
   n148 = "Constraint - dyn_hair_a (PP)"(ComponentPose=@n146, bAlphaBoolEnabled=@n147, ConstraintWeights_0=0.7, ConstraintWeights_1=0.3) {
    alphaInputType=Bool
    alphaBoolBlend {blendOption=Linear}
    boneToModify {boneName=dyn_hair_a}
    constraintSetup [{targetBone {boneName=head}}, {targetBone {boneName=spine_05}}]
    constraintWeights [1,1]
   }
   n149 = "Constraint - dyn_hair_b (PP)"(ComponentPose=@n148, bAlphaBoolEnabled=@n147, ConstraintWeights_0=0.3, ConstraintWeights_1=0.7) {
    alphaInputType=Bool
    alphaBoolBlend {blendOption=Linear}
    boneToModify {boneName=dyn_hair_b}
    constraintSetup [{targetBone {boneName=head}}, {targetBone {boneName=spine_05}}]
    constraintWeights [1,1]
   }
   n150 = "Constraint - dyn_hair_c (PPP)"(ComponentPose=@n149, bAlphaBoolEnabled=@n147, ConstraintWeights_0=0.2, ConstraintWeights_1=0.2, ConstraintWeights_2=0.6) {
    alphaInputType=Bool
    alphaBoolBlend {blendOption=Linear}
    boneToModify {boneName=dyn_hair_c}
    constraintSetup [{targetBone {boneName=pelvis}}, {targetBone {boneName=spine_04}}, {targetBone {boneName=spine_05}}]
    constraintWeights [1,1,1]
   }
   n151 = "Constraint - dyn_hair_d (PPP)"(ComponentPose=@n150, bAlphaBoolEnabled=@n147, ConstraintWeights_0=0.7, ConstraintWeights_1=0.2, ConstraintWeights_2=0.1) {
    alphaInputType=Bool
    alphaBoolBlend {blendOption=Linear}
    boneToModify {boneName=dyn_hair_d}
    constraintSetup [{targetBone {boneName=pelvis}}, {targetBone {boneName=spine_04}}, {targetBone {boneName=spine_05}}]
    constraintWeights [1,1,1]
   }
   n152 = "Constraint - dyn_hair_e (PP)"(ComponentPose=@n151, bAlphaBoolEnabled=@n147, ConstraintWeights_0=0.8, ConstraintWeights_1=0.2) {
    alphaInputType=Bool
    alphaBoolBlend {blendOption=Linear}
    boneToModify {boneName=dyn_hair_e}
    constraintSetup [{targetBone {boneName=pelvis}}, {targetBone {boneName=spine_02}}]
    constraintWeights [1,1]
   }
   n153 = Get hairConstraintON()
   n154 = "Transform (Modify) Bone - Bone: dyn_hair_b"!(ComponentPose=@n152, bAlphaBoolEnabled:bool=@n153, Translation="0,0,-5", Rotation="5,0,0", Scale="1,1,1") -> (Pose) {
    alphaInputType=Bool
    rotationMode=BMM_Additive
    rotationSpace=BCS_BoneSpace
    translationMode=BMM_Additive
    translationSpace=BCS_BoneSpace
    alphaBoolBlend {blendOption=Linear}
    alphaScaleBiasClamp {bMapRange=true, inRange {max=0, min=-25}, outRange {max=0, min=1}}
    boneToModify {boneName=dyn_hair_b}
    rotation {pitch=5}
    translation {z=-5}
   }
   n155 = "Constraint - dyn_hair_small_a (PPP)"(ComponentPose=@n154, bAlphaBoolEnabled=@n153, ConstraintWeights_0=0.2, ConstraintWeights_1=0.7, ConstraintWeights_2=0.1) {
    alphaInputType=Bool
    alphaBoolBlend {blendOption=Linear}
    boneToModify {boneName=dyn_hair_small_a}
    constraintSetup [{targetBone {boneName=dyn_hair_c}}, {targetBone {boneName=dyn_hair_b}}, {targetBone {boneName=dyn_hair_a}}]
    constraintWeights [1,1,1]
   }
   n156 = "Constraint - dyn_hair_small_b (PP)"(ComponentPose=@n155, bAlphaBoolEnabled=@n153, ConstraintWeights_0=0.3, ConstraintWeights_1=0.7) {
    alphaInputType=Bool
    alphaBoolBlend {blendOption=Linear}
    boneToModify {boneName=dyn_hair_small_b}
    constraintSetup [{targetBone {boneName=dyn_hair_d}}, {targetBone {boneName=dyn_hair_c}}]
    constraintWeights [1,1]
   }
   n157 = "Constraint - dyn_hair_small_c (PPP)"(ComponentPose=@n156, bAlphaBoolEnabled=@n153, ConstraintWeights_0=0.1, ConstraintWeights_1=0.8, ConstraintWeights_2=0.1) {
    alphaInputType=Bool
    alphaBoolBlend {blendOption=Linear}
    boneToModify {boneName=dyn_hair_small_c}
    constraintSetup [{targetBone {boneName=dyn_hair_d}}, {targetBone {boneName=dyn_hair_c}}, {targetBone {boneName=dyn_hair_b}}]
    constraintWeights [1,1,1]
   }
   n158 = Get IsShortBackblingOn()
   n159 = "Transform (Modify) Bone - Bone: dyn_hair_band_base"(ComponentPose=@n157, bAlphaBoolEnabled=@n158, Translation="0,1,0", Rotation="14,0,10", Scale="0.8,1,1") {
    alphaInputType=Bool
    rotationMode=BMM_Additive
    rotationSpace=BCS_BoneSpace
    scaleMode=BMM_Additive
    scaleSpace=BCS_BoneSpace
    translationMode=BMM_Additive
    translationSpace=BCS_BoneSpace
    alphaBoolBlend {blendOption=Linear}
    alphaScaleBiasClamp {bMapRange=true, inRange {max=0, min=-25}, outRange {max=0, min=1}}
    boneToModify {boneName=dyn_hair_band_base}
    rotation {pitch=14, roll=10}
    scale {x=0.8}
    translation {y=1}
   }
   n160 = "Transform (Modify) Bone - Bone: dyn_hair_a"!(ComponentPose=@n159, bAlphaBoolEnabled:bool=@n158, Translation="0,2,0", Rotation="0,-10,0", Scale="1,1,1") -> (Pose) {
    alphaInputType=Bool
    rotationMode=BMM_Additive
    rotationSpace=BCS_BoneSpace
    scaleMode=BMM_Additive
    scaleSpace=BCS_BoneSpace
    translationMode=BMM_Additive
    translationSpace=BCS_BoneSpace
    alphaBoolBlend {blendOption=Linear}
    alphaScaleBiasClamp {bMapRange=true, inRange {max=0, min=-25}, outRange {max=0, min=1}}
    boneToModify {boneName=dyn_hair_a}
    rotation {yaw=-10}
    translation {y=2}
   }
   n161 = "Transform (Modify) Bone - Bone: dyn_hair_b"(ComponentPose=@n160, bAlphaBoolEnabled=@n158, Translation="-2,0,0", Rotation="-5,0,0", Scale="1,1,1") {
    alphaInputType=Bool
    rotationMode=BMM_Additive
    rotationSpace=BCS_BoneSpace
    scaleMode=BMM_Additive
    scaleSpace=BCS_BoneSpace
    translationMode=BMM_Additive
    translationSpace=BCS_BoneSpace
    alphaBoolBlend {blendOption=Linear}
    alphaScaleBiasClamp {bMapRange=true, inRange {max=0, min=-25}, outRange {max=0, min=1}}
    boneToModify {boneName=dyn_hair_b}
    rotation {pitch=-5}
    translation {x=-2}
   }
   n162 = Get IsShortBackblingOn()
   n163 = "Transform (Modify) Bone - Bone: dyn_hair_small_a"(ComponentPose=@n161, bAlphaBoolEnabled=@n162, Translation="0,0,0", Rotation="0,0,0", Scale="0.85,0.85,0.85") {
    alphaInputType=Bool
    rotationMode=BMM_Additive
    rotationSpace=BCS_BoneSpace
    scaleMode=BMM_Additive
    scaleSpace=BCS_BoneSpace
    translationMode=BMM_Additive
    translationSpace=BCS_BoneSpace
    alphaBoolBlend {blendOption=Linear}
    alphaScaleBiasClamp {bMapRange=true, inRange {max=0, min=-25}, outRange {max=0, min=1}}
    boneToModify {boneName=dyn_hair_small_a}
    scale=<0.85 0.85 0.85>
   }
   n164 = "Transform (Modify) Bone - Bone: dyn_hair_small_b"(ComponentPose=@n163, bAlphaBoolEnabled=@n162, Translation="0,-4,0", Rotation="0,5,0", Scale="0.8,1,1") {
    alphaInputType=Bool
    rotationMode=BMM_Additive
    rotationSpace=BCS_BoneSpace
    scaleMode=BMM_Additive
    scaleSpace=BCS_BoneSpace
    translationMode=BMM_Additive
    translationSpace=BCS_BoneSpace
    alphaBoolBlend {blendOption=Linear}
    alphaScaleBiasClamp {bMapRange=true, inRange {max=0, min=-25}, outRange {max=0, min=1}}
    boneToModify {boneName=dyn_hair_small_b}
    rotation {yaw=5}
    scale {x=0.8}
    translation {y=-4}
   }
   n165 = "Transform (Modify) Bone - Bone: dyn_hair_small_c"(ComponentPose=@n164, bAlphaBoolEnabled=@n162, Translation="0,-3,0", Rotation="0,0,0", Scale="0.75,1,1") {
    alphaInputType=Bool
    rotationMode=BMM_Additive
    rotationSpace=BCS_BoneSpace
    scaleMode=BMM_Additive
    scaleSpace=BCS_BoneSpace
    translationMode=BMM_Additive
    translationSpace=BCS_BoneSpace
    alphaBoolBlend {blendOption=Linear}
    alphaScaleBiasClamp {bMapRange=true, inRange {max=0, min=-25}, outRange {max=0, min=1}}
    boneToModify {boneName=dyn_hair_small_c}
    scale {x=0.75}
    translation {y=-3}
   }
   n166 = "Transform (Modify) Bone - Bone: dyn_hair_c"!(ComponentPose=@n165, bAlphaBoolEnabled:bool=@n162, Translation="-1.5,0,0", Rotation="-10,0,0", Scale="0.5,0.8,0.8") -> (Pose) {
    alphaInputType=Bool
    rotationMode=BMM_Additive
    rotationSpace=BCS_BoneSpace
    scaleMode=BMM_Additive
    scaleSpace=BCS_BoneSpace
    translationMode=BMM_Additive
    translationSpace=BCS_BoneSpace
    alphaBoolBlend {blendOption=Linear}
    alphaScaleBiasClamp {bMapRange=true, inRange {max=0, min=-25}, outRange {max=0, min=1}}
    boneToModify {boneName=dyn_hair_c}
    rotation {pitch=-10}
    scale=<0.5 0.8 0.8>
    translation {x=-1.5}
   }
   n167 = "Transform (Modify) Bone - Bone: dyn_hair_d"(ComponentPose=@n166, bAlphaBoolEnabled=@n162, Translation="-1.5,-2,0", Rotation="0,-5,0", Scale="0.5,0.8,0.8") {
    alphaInputType=Bool
    rotationMode=BMM_Additive
    rotationSpace=BCS_BoneSpace
    scaleMode=BMM_Additive
    scaleSpace=BCS_BoneSpace
    translationMode=BMM_Additive
    translationSpace=BCS_BoneSpace
    alphaBoolBlend {blendOption=Linear}
    alphaScaleBiasClamp {bMapRange=true, inRange {max=0, min=-25}, outRange {max=0, min=1}}
    boneToModify {boneName=dyn_hair_d}
    rotation {yaw=-5}
    scale=<0.5 0.8 0.8>
    translation=<-1.5 -2>
   }
   n168 = "Transform (Modify) Bone - Bone: dyn_hair_e"(ComponentPose=@n167, bAlphaBoolEnabled=@n162, Translation="-1.5,0,0", Rotation="0,0,0", Scale="0.4,0.8,0.8") {
    alphaInputType=Bool
    rotationMode=BMM_Additive
    rotationSpace=BCS_BoneSpace
    scaleMode=BMM_Additive
    scaleSpace=BCS_BoneSpace
    translationMode=BMM_Additive
    translationSpace=BCS_BoneSpace
    alphaBoolBlend {blendOption=Linear}
    alphaScaleBiasClamp {bMapRange=true, inRange {max=0, min=-25}, outRange {max=0, min=1}}
    boneToModify {boneName=dyn_hair_e}
    scale=<0.4 0.8 0.8>
    translation {x=-1.5}
   }
   n169 = Get backpackOnHeadUpDn()
   n170 = "Transform (Modify) Bone - Bone: dyn_hair_band_base"!(ComponentPose=@n168, Alpha:float=@n169, Translation="0,0,0", Rotation="-12,-30,-10", Scale="1.2,1.05,1.05") -> (Pose) {
    rotationMode=BMM_Additive
    rotationSpace=BCS_BoneSpace
    scaleMode=BMM_Additive
    scaleSpace=BCS_BoneSpace
    translationMode=BMM_Additive
    translationSpace=BCS_BoneSpace
    alphaBoolBlend {blendOption=Linear}
    alphaScaleBiasClamp {bMapRange=true, inRange {max=0, min=-25}, outRange {max=0, min=1}}
    boneToModify {boneName=dyn_hair_band_base}
    rotation=<rot -12 -30 -10>
    scale=<1.2 1.05 1.05>
   }
   n171 = "Transform (Modify) Bone - Bone: dyn_hair_a"!(ComponentPose=@n170, Alpha:float=@n169, Translation="0,1,0", Rotation="10,-27,0", Scale="1,1,1") -> (Pose) {
    rotationMode=BMM_Additive
    rotationSpace=BCS_BoneSpace
    scaleMode=BMM_Additive
    scaleSpace=BCS_BoneSpace
    translationMode=BMM_Additive
    translationSpace=BCS_BoneSpace
    alphaBoolBlend {blendOption=Linear}
    alphaScaleBiasClamp {bMapRange=true, inRange {max=0, min=-25}, outRange {max=0, min=1}}
    boneToModify {boneName=dyn_hair_a}
    rotation {pitch=10, yaw=-27}
    translation {y=1}
   }
   n172 = "Transform (Modify) Bone - Bone: dyn_hair_b"!(ComponentPose=@n171, Alpha:float=@n169, Translation="1,-3,0", Rotation="0,5,0", Scale="1,1,1") -> (Pose) {
    rotationMode=BMM_Additive
    rotationSpace=BCS_BoneSpace
    scaleMode=BMM_Additive
    scaleSpace=BCS_BoneSpace
    translationMode=BMM_Additive
    translationSpace=BCS_BoneSpace
    alphaBoolBlend {blendOption=Linear}
    alphaScaleBiasClamp {bMapRange=true, inRange {max=0, min=-25}, outRange {max=0, min=1}}
    boneToModify {boneName=dyn_hair_b}
    rotation {yaw=5}
    translation=<1 -3>
   }
   n173 = "Transform (Modify) Bone - Bone: dyn_hair_small_b"!(ComponentPose=@n172, Alpha:float=@n169, Translation="0,-1.5,0", Rotation="0,0,0", Scale="1,1,1") -> (Pose) {
    rotationMode=BMM_Additive
    rotationSpace=BCS_BoneSpace
    scaleMode=BMM_Additive
    scaleSpace=BCS_BoneSpace
    translationMode=BMM_Additive
    translationSpace=BCS_BoneSpace
    alphaBoolBlend {blendOption=Linear}
    alphaScaleBiasClamp {bMapRange=true, inRange {max=0, min=-25}, outRange {max=0, min=1}}
    boneToModify {boneName=dyn_hair_small_b}
    translation {y=-1.5}
   }
   n174 = "Transform (Modify) Bone - Bone: dyn_hair_c"!(ComponentPose=@n173, Alpha:float=@n169, Translation="0,0,0", Rotation="-20,0,0", Scale="1,1,1") -> (Pose) {
    rotationMode=BMM_Additive
    rotationSpace=BCS_BoneSpace
    scaleMode=BMM_Additive
    scaleSpace=BCS_BoneSpace
    translationMode=BMM_Additive
    translationSpace=BCS_BoneSpace
    alphaBoolBlend {blendOption=Linear}
    alphaScaleBiasClamp {bMapRange=true, inRange {max=0, min=-25}, outRange {max=0, min=1}}
    boneToModify {boneName=dyn_hair_c}
    rotation {pitch=-20}
   }
   n175 = Get backpackOnHeadUpDn()
   n176 = "Transform (Modify) Bone - Bone: dyn_hair_a"(ComponentPose=@n174, Alpha=@n175, Translation="0,1,0", Rotation="0,0,0", Scale="1,1,1") {
    rotationMode=BMM_Additive
    rotationSpace=BCS_BoneSpace
    scaleMode=BMM_Additive
    scaleSpace=BCS_BoneSpace
    translationMode=BMM_Additive
    translationSpace=BCS_BoneSpace
    alphaBoolBlend {blendOption=Linear}
    alphaScaleBiasClamp {bMapRange=true, inRange {max=15}}
    boneToModify {boneName=dyn_hair_a}
    translation {y=1}
   }
   n177 = "Transform (Modify) Bone - Bone: dyn_hair_b"(ComponentPose=@n176, Alpha=@n175, Translation="0,0,0", Rotation="0,2,0", Scale="1,1,1") {
    rotationMode=BMM_Additive
    rotationSpace=BCS_BoneSpace
    scaleMode=BMM_Additive
    scaleSpace=BCS_BoneSpace
    translationMode=BMM_Additive
    translationSpace=BCS_BoneSpace
    alphaBoolBlend {blendOption=Linear}
    alphaScaleBiasClamp {bMapRange=true, inRange {max=15}}
    boneToModify {boneName=dyn_hair_b}
    rotation {yaw=2}
   }
   n178 = "Transform (Modify) Bone - Bone: dyn_hair_small_b"(ComponentPose=@n177, Alpha=@n175, Translation="0,0,0", Rotation="0,5,0", Scale="1,1,1") {
    rotationMode=BMM_Additive
    rotationSpace=BCS_BoneSpace
    scaleMode=BMM_Additive
    scaleSpace=BCS_BoneSpace
    translationMode=BMM_Additive
    translationSpace=BCS_BoneSpace
    alphaBoolBlend {blendOption=Linear}
    alphaScaleBiasClamp {bMapRange=true, inRange {max=15}}
    boneToModify {boneName=dyn_hair_small_b}
    rotation {yaw=5}
   }
   n179 = "Transform (Modify) Bone - Bone: dyn_hair_d"(ComponentPose=@n178, bAlphaBoolEnabled=Get bIsParachuteOpened(), Translation="0,0,0", Rotation="0,-30,0", Scale="1,1,1") {
    alphaInputType=Bool
    rotationMode=BMM_Additive
    rotationSpace=BCS_BoneSpace
    scaleMode=BMM_Additive
    scaleSpace=BCS_BoneSpace
    translationMode=BMM_Additive
    translationSpace=BCS_BoneSpace
    alphaBoolBlend {blendOption=Linear}
    alphaScaleBiasClamp {bMapRange=true, inRange {max=5, min=-15}}
    boneToModify {boneName=dyn_hair_d}
    rotation {yaw=-30}
   }
   n180 = Get isBistroGlider()
   n181 = "Transform (Modify) Bone - Bone: dyn_hair_a"(ComponentPose=@n179, Alpha=@n180, Translation="0,0,0", Rotation="0,35,0", Scale="1,1,1") {
    rotationMode=BMM_Additive
    rotationSpace=BCS_BoneSpace
    translationMode=BMM_Additive
    translationSpace=BCS_BoneSpace
    alphaBoolBlend {blendOption=Linear}
    alphaScaleBiasClamp {bMapRange=true}
    boneToModify {boneName=dyn_hair_a}
    rotation {yaw=35}
   }
   n182 = "Transform (Modify) Bone - Bone: dyn_hair_b"(ComponentPose=@n181, Alpha=@n180, Translation="0,0,0", Rotation="0,20,0", Scale="1,1,1") {
    rotationMode=BMM_Additive
    rotationSpace=BCS_BoneSpace
    translationMode=BMM_Additive
    translationSpace=BCS_BoneSpace
    alphaBoolBlend {blendOption=Linear}
    alphaScaleBiasClamp {bMapRange=true}
    boneToModify {boneName=dyn_hair_b}
    rotation {yaw=20}
   }
   n183 = "Transform (Modify) Bone - Bone: dyn_hair_a_base"(ComponentPose=@n182, Alpha=@n180, Translation="0,0,0", Rotation="0,10,0", Scale="1,1,1") {
    rotationMode=BMM_Additive
    rotationSpace=BCS_BoneSpace
    translationMode=BMM_Additive
    translationSpace=BCS_BoneSpace
    alphaBoolBlend {blendOption=Linear}
    alphaScaleBiasClamp {bMapRange=true}
    boneToModify {boneName=dyn_hair_a_base}
    rotation {yaw=10}
   }
   n184 = "Transform (Modify) Bone - Bone: dyn_hair_c"(ComponentPose=@n183, Alpha=Get head_fwdBwd_rot(), Translation="0,2,0", Rotation="0,7,0", Scale="1,1,1") {
    rotationMode=BMM_Additive
    rotationSpace=BCS_BoneSpace
    scaleMode=BMM_Additive
    scaleSpace=BCS_BoneSpace
    translationMode=BMM_Additive
    translationSpace=BCS_BoneSpace
    alphaBoolBlend {blendOption=Linear}
    alphaScaleBiasClamp {bMapRange=true, inRange {max=5, min=-15}}
    boneToModify {boneName=dyn_hair_c}
    rotation {yaw=7}
    translation {y=2}
   }
   n185 = "Transform (Modify) Bone - Bone: tail_base"(ComponentPose=@n184, Alpha=1, Translation="0,0,0", Rotation=Get Tail_Base_Rot(), Scale="1,1,1") {rotationMode=BMM_Replace, rotationSpace=BCS_BoneSpace, alphaBoolBlend {blendOption=Linear}, boneToModify {boneName=tail_base}}
   n186 = "Transform (Modify) Bone - Bone: tail_1"(ComponentPose=@n185, Alpha=1, Translation="0,0,0", Rotation=Get Tail_1_Rot(), Scale="1,1,1") {
    rotationMode=BMM_Replace
    rotationSpace=BCS_BoneSpace
    translationSpace=BCS_BoneSpace
    alphaBoolBlend {blendOption=Linear}
    boneToModify {boneName=tail_1}
   }
   n187 = "Transform (Modify) Bone - Bone: tail_2"(ComponentPose=@n186, Alpha=1, Translation="0,0,0", Rotation=Get Tail_2_Rot(), Scale="1,1,1") {
    rotationMode=BMM_Replace
    rotationSpace=BCS_BoneSpace
    translationSpace=BCS_BoneSpace
    alphaBoolBlend {blendOption=Linear}
    boneToModify {boneName=tail_2}
   }
   n188 = "Transform (Modify) Bone - Bone: tail_3"(ComponentPose=@n187, Alpha=1, Translation="0,0,0", Rotation=Get Tail_3_Rot(), Scale="1,1,1") {
    rotationMode=BMM_Replace
    rotationSpace=BCS_BoneSpace
    translationSpace=BCS_BoneSpace
    alphaBoolBlend {blendOption=Linear}
    boneToModify {boneName=tail_3}
   }
   Save cached pose 'SharedCachePose'(Pose=Component To Local(ComponentPose=@n188)) {cacheName=SharedCachePose}
   Use cached pose 'NO Backbling Cache'()
   Use cached pose 'Backbling Cache'()
   Use cached pose 'Backbling Cache'()
   Use cached pose 'SimCachPose'()
   NOT Boolean(A=false)
   NOT Boolean(A=Get bRigidBodyEnabled())
   NOT Boolean(A=Get hairConstraintON())
   AND Boolean(A=@n53, B=Get IsLongBackblingOn() -> (IsLongBackblingOn:bool))
   Get IsBackblingOnAlpha() -> (IsBackblingOnAlpha:double)
   Get IsBackblingOn()
   Get IsBackblingOn()
   Get IsBackblingOn()
   Get IsBackblingOn()
   Get IsBackblingOn()
   Get IsBackblingOn()
  ]}
 EventGraph {nodes [
   Delay(Duration=0.5)
   > ResetDynamics(InTeleportType=TeleportPhysics)
   > ResetDynamics(InTeleportType=ResetPhysics)
   Event BlueprintUpdateAnimation()
   > n1 = Sequence!(execute) -> (then_0, then_1, then_2:Exec, then_3:Exec, then_4:Exec, then_5:Exec, then_6:Exec, then_7:Exec, then_8:Exec)
   n2 = Get bIsFalling()
   n3 = Get bIsJumping()
   n4 = Branch(execute=@n1.then_2, Condition=AND Boolean(A=Get IsLongBackblingOn(), B=OR Boolean!(A=@n3, B=@n2) -> (ReturnValue)))
   > Set isJumpPoseFixNeeded(isJumpPoseFixNeeded:double=1) -> (Output_Get:double)
   Set isJumpPoseFixNeeded(execute=@n4.else, isJumpPoseFixNeeded=0)
   Branch(execute=@n1.then_7, Condition=Get IsTailOn())
   > n5 = Cast To FortPlayerPawnAthena(Object=TryGetPawnOwner())
   n6 = GetSkeletalMeshForPartType(self=@n5, PartType=MiscOrTail)
   > Set Tail_Base_Rot(Tail_Base_Rot:Rotator=BreakTransform(InTransform=GetDeltaTransformFromRefPose(self=@n6, BoneName=tail_base, BaseName=root)).Rotation) -> (Output_Get:Rotator)
   > Set Tail_1_Rot(Tail_1_Rot:Rotator=BreakTransform(InTransform=GetDeltaTransformFromRefPose(self=@n6, BoneName=tail_1, BaseName=tail_base)).Rotation) -> (Output_Get:Rotator)
   > Set Tail_2_Rot(Tail_2_Rot:Rotator=BreakTransform(InTransform=GetDeltaTransformFromRefPose(self=@n6, BoneName=tail_2, BaseName=tail_1)).Rotation) -> (Output_Get:Rotator)
   > Set Tail_3_Rot(Tail_3_Rot:Rotator=BreakTransform(InTransform=GetDeltaTransformFromRefPose(self=@n6, BoneName=tail_2, BaseName=tail_1)).Rotation) -> (Output_Get:Rotator)
   n7 = Cast To FortPlayerPawn(execute=@n1.then_8, Object=GetOwningActor())
   n8 = Get ParachuteAttachment!(self:Fort Player Pawn Object Reference=@n7) -> (ParachuteAttachment:Fort Player Parachute Object Reference)
   > n9 = Is Valid(InputObject=@n8)
   Set backbling_glider(execute=[@n7.CastFailed, @n9.Is Not Valid, @n11.Is Not Valid], backbling_glider:bool=false) -> (Output_Get:bool)
   n10 = Get ParachuteItemDef!(self:Fort Player Parachute Object Reference=@n8) -> (ParachuteItemDef:Athena Glider Item Definition Object Reference)
   n11 = Is Valid(exec=@n9.Is Valid, InputObject=@n10)
   n12 = AND Boolean(
    A=HasAnyMatchingGameplayTags!(
     self:Gameplay Tag Asset Interface Interface=@n10
     TagContainer:"Gameplay Tag Container Structure (by ref)"=Get Tag Container() -> (Tag Container:Gameplay Tag Container Structure)
    ) -> (ReturnValue:bool)
    B=OR Boolean(A=Get bIsParachuteOpened(), B=Get bIsInFrontEnd())
   )
   > Set backbling_glider(backbling_glider=@n12)
   > n13 = Branch(Condition=@n12)
   > Set isBistroGlider(isBistroGlider:double=1) -> (Output_Get:double)
   Set isBistroGlider(execute=@n13.else, isBistroGlider=0)
   n14 = GetDeltaTransformFromRefPose(self=GetOwnerSkeletalMeshComponent(), BoneName=head, BaseName=neck_01)
   n15 = Set head_fwdBwd_rot(execute=@n1.then_0, head_fwdBwd_rot:double=BreakRotator(InRot=BreakTransform(InTransform=@n14).Rotation).Yaw) -> (Output_Get:double)
   > n16 = Branch(Condition=Get IsBackblingOn())
   > Set backpackOnHeadUpDn(backpackOnHeadUpDn:double=@n15) -> (Output_Get:double)
   Set backpackOnHeadUpDn(execute=@n16.else, backpackOnHeadUpDn=0)
   n17 = Get bRigidBodyEnabled()
   n18 = Branch(execute=@n1.then_4, Condition=OR Boolean(A=Get IsShortBackblingOn(), B=NOT Boolean(A=@n17)))
   > Set hairConstraintON(hairConstraintON:bool=true) -> (Output_Get:bool)
   Set hairConstraintON(execute=@n18.else, hairConstraintON=false)
   n19 = Get IsLongBackblingOn()
   n20 = Branch(
    execute=@n1.then_5
    Condition=AND Boolean(A=Get IsBackblingOn(), B="Not Equal (Boolean)"!(A:bool=@n19, B:bool=true) -> (ReturnValue:bool) {is=title K2Node_CallFunction})
   )
   > Set rigidBodyAlpha(rigidBodyAlpha:double=0) -> (Output_Get:double)
   > Set skyDiveFloatPoseFixAlpha(skyDiveFloatPoseFixAlpha:double=0) -> (Output_Get:double)
   n21 = Sequence!(execute=@n20.else) -> (then_0, then_1)
   > n22 = Branch(Condition=Get bIsSkyDiving())
   > n23 = Branch(Condition=Get IsShortBackblingOn())
   > Set rigidBodyAlpha(rigidBodyAlpha=0.3)
   Set rigidBodyAlpha(execute=@n23.else, rigidBodyAlpha=0.6)
   Set rigidBodyAlpha(execute=@n22.else, rigidBodyAlpha=1)
   n24 = Branch(execute=@n21.then_1, Condition=OR Boolean(A=Get bIsSkydiveDiveMode(), B=Get bIsSkydiveFloating()))
   > n25 = Branch(Condition=Get IsBackblingOn())
   > Set skyDiveFloatPoseFixAlpha(skyDiveFloatPoseFixAlpha=1)
   Set skyDiveFloatPoseFixAlpha(execute=@n25.else, skyDiveFloatPoseFixAlpha=0.5)
   Set skyDiveFloatPoseFixAlpha(execute=@n24.else, skyDiveFloatPoseFixAlpha=0)
   Is Valid(exec=@n1.then_3, InputObject=Get Pawn Owner() -> (Pawn Owner:Player Pawn Athena Object Reference))
   n26 = CheckForBackbling(LongBackblingThreshold=40)
   n27 = OR Boolean(A=Get backbling_glider() -> (backbling_glider:bool), B=@n26)
   > Branch(Condition=@n27)
   > n28 = Sequence!(execute) -> (then_0, then_1, then_2:Exec)
   > n29 = Set IsBackblingOn(IsBackblingOn:bool=@n27) -> (Output_Get:bool)
   > Set IsBackblingOnAlpha(IsBackblingOnAlpha:double="To Float (Boolean)"(InBool=@n29)) -> (Output_Get:double)
   n30 = Branch(execute=@n28.then_1, Condition=OR Boolean(A=@n26.bIsBackblingACape, B=@n26.bIsBackblingLong))
   > Set IsLongBackblingOn(IsLongBackblingOn:bool=true) -> (Output_Get:bool)
   > Set IsLongBackblingOnAlpha(IsLongBackblingOnAlpha:double=1) -> (Output_Get:double)
   > n31 = Set IsShortBackblingOn(IsShortBackblingOn:bool=false) -> (Output_Get:bool)
   > Set IsShortBackblingOnAlpha(IsShortBackblingOnAlpha:double="To Float (Boolean)"(InBool=@n31)) -> (Output_Get:double)
   > Set IsTailOn(IsTailOn:bool=false) -> (Output_Get:bool)
   Set IsLongBackblingOn(execute=@n30.else, IsLongBackblingOn=false)
   > Set IsLongBackblingOnAlpha(IsLongBackblingOnAlpha=0)
   > n32 = Set IsShortBackblingOn(IsShortBackblingOn=true)
   > Set IsShortBackblingOnAlpha(IsShortBackblingOnAlpha="To Float (Boolean)"(InBool=@n32))
   > Set IsTailOn(IsTailOn=true)
   n33 = Random Vector Function!(
    execute=@n1.then_6
    CharacterPartAnimInstance:Custom Character Part Anim Instance Object Reference
    SkelMeshComponent:Mesh Component Object Reference=GetOwnerSkeletalMeshComponent()
    JointName:Name=pelvis
    GameTimeInSeconds:double=GetGameTimeInSeconds()
    TimeOffset:double=0
    PerlinScale:double=2
    Frequency:Vector="2,2.5,2"
    Amplitude:Vector="10,120,10"
    Offset:Vector="0,-200,0"
   ) -> (then, VectorOut:Vector)
   n34 = BreakVector(InVec=@n33)
   > Set Gravity Override(Gravity Override=MakeVector(X=@n34, Y=@n34.Y, Z=-980))
   n35 = GetOwnerDirection!(self:Custom Character Part Anim Instance Object Reference) -> (ReturnValue:float)
   Set RelaxSpeed(
    execute=@n1.then_1
    RelaxSpeed:double=MapRangeClamped(Value="Absolute (Float)"(A=@n35), InRangeA=85, InRangeB=110, OutRangeA=1.5, OutRangeB=4)
   ) -> (Output_Get:double)
   > n36 = Branch(Condition=Get bIsSkyDiving())
   > Set TrailAlpha(TrailAlpha:double=0.5) -> (Output_Get:double)
   n37 = Branch(execute=@n36.else, Condition=OR Boolean(A=Get bIsJumping(), B=Get bIsFalling()))
   > Set TrailAlpha(TrailAlpha=0.75)
   Set TrailAlpha(execute=@n37.else, TrailAlpha=1)
   Event OnCharacterCustomizationLoaded()
   > n38 = Branch(Condition=CheckForTail!(self:Custom Character Part Anim Instance Object Reference) -> (bHasTail:bool))
   > Cast To FortPlayerPawnAthena(Object=TryGetPawnOwner())
   Set IsTailOn(execute=@n38.else, IsTailOn=false)
   Event BlueprintInitializeAnimation()
   > Set Body_Mesh(Body_Mesh=GetOwnerSkeletalMeshComponent())
   > n39 = Cast To PlayerPawn_Athena(Object=TryGetPawnOwner())
   > n40 = Set Pawn Owner(Pawn Owner:Player Pawn Athena Object Reference=@n39) -> (Output_Get:Player Pawn Athena Object Reference)
   > Set Head_Mesh(Head_Mesh:Skeletal Mesh Component Object Reference=GetSkeletalMeshForPartType(self=@n40, PartType=Head)) -> (Output_Get:Skeletal Mesh Component Object Reference)
   > n41 = Branch(Condition=CheckForTail())
   > Set IsTailOn(IsTailOn=true)
   Set IsTailOn(execute=@n41.else, IsTailOn=false)
   Reroute Node()
   n42 = GetOwnerDirection()
   > Set RelaxSpeed(RelaxSpeed=MapRangeClamped(Value="Absolute (Float)"(A=@n42), InRangeA=85, InRangeB=110, OutRangeA=1.5, OutRangeB=4))
   > n43 = Branch(Condition=Get bIsSkyDiving())
   > Set TrailAlpha(TrailAlpha=0.5)
   n44 = Branch(execute=@n43.else, Condition=OR Boolean(A=Get bIsJumping(), B=Get bIsFalling()))
   > Set TrailAlpha(TrailAlpha=0.75)
   Set TrailAlpha(execute=@n44.else, TrailAlpha=1)
   Reroute Node()
   n45 = GetDeltaTransformFromRefPose(self=GetOwnerSkeletalMeshComponent(), BoneName=head, BaseName=neck_01)
   > Set head_fwdBwd_rot(head_fwdBwd_rot=BreakRotator(InRot=BreakTransform(InTransform=@n45).Rotation).Yaw)
   Set Tail_Mesh(Tail_Mesh:Skeletal Mesh Component Object Reference=GetSkeletalMeshForPartType(PartType=MiscOrTail)) -> (Output_Get:Skeletal Mesh Component Object Reference)
   > Set IsTailOn(IsTailOn=true)
   Do Once()
   Set Tail_Mesh(Tail_Mesh=GetSkeletalMeshForPartType(self=@n40, PartType=MiscOrTail))
  ]}
]
variables [
 backbling_glider {type=boolean}
 backpackOnHeadUpDn {type=number}
 Body_Mesh {type=SkeletalMeshComponent}
 Gravity Override {type=Vector}
 hairConstraintON {type=boolean}
 head_fwdBwd_rot {type=number}
 Head_Mesh {type=SkeletalMeshComponent}
 IsBackblingOn {type=boolean}
 IsBackblingOnAlpha {type=number}
 isBistroGlider {type=number}
 isJumpPoseFixNeeded {type=number}
 IsLongBackblingOn {type=boolean}
 IsLongBackblingOnAlpha {type=number}
 IsShortBackblingOn {type=boolean}
 IsShortBackblingOnAlpha {type=number}
 IsTailOn {type=boolean}
 Pawn Owner {type=PlayerPawn_Athena_C}
 RelaxSpeed {type=number}
 rigidBodyAlpha {type=number}
 skyDiveFloatPoseFixAlpha {type=number}
 Tag Container {type=GameplayTagContainer, default {gameplayTags [{tagName=Cosmetics.Glider.BackpackGlider}]}}
 Tail_1_Rot {type=Rotator}
 Tail_2_Rot {type=Rotator}
 Tail_3_Rot {type=Rotator}
 Tail_Base_Rot {type=Rotator}
 Tail_Mesh {type=SkeletalMeshComponent}
 Tail_Root_Rot {type=Rotator}
 TrailAlpha {type=number}
]
